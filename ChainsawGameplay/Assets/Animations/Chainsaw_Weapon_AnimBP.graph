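# Animation Blueprint for the chainsaw weapon mesh, animating its equip motion and spinning chain.
# Equip SM picks slide, sprint or idle equip clips from movement booleans (falling uses slide), layered onto the ref pose at pull_cord.
# ChainSpinning SM is layered at tooth_main, replaying a non-looping clip by alternating two identical states through conduits while bIsSpinning holds.
# The EventGraph is empty, so no state flags are set there.
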AnimBlueprint /ChainsawGameplay/Assets/Animations/Chainsaw_Weapon_AnimBP
target_skeleton=/ChainsawGameplay/Assets/Meshes/Chainsaw_Sharp_Skeleton
graphs [
 AnimGraph {nodes [
   n1 = Local Space Ref Pose()
   AnimGraphNode_StateMachine_1 = Equip SM() {is=title AnimGraphNode_StateMachine}
   AnimGraphNode_StateMachine_0 = ChainSpinning SM() {is=title AnimGraphNode_StateMachine}
   Output Pose(Result=Layered blend per bone(
     BasePose=Layered blend per bone(BasePose=@n1, BlendPoses_0=@AnimGraphNode_StateMachine_1) {bMeshSpaceRotationBlend=true, layerSetup [{branchFilters [{boneName=pull_cord}]}]}
     BlendPoses_0=@AnimGraphNode_StateMachine_0
    ) {layerSetup [{branchFilters [{boneName=tooth_main}]}]})
  ]}
 ChainSpinning SM {
  entry=Wait_1
  path=AnimGraph.AnimGraphNode_StateMachine_0.ChainSpinning SM
  ids=id
  nodes [
   AnimStateConduitNode_0 = IsSpinning? 1() {is=title AnimStateConduitNode}
   AnimStateConduitNode_1 = IsSpinning? 2() {is=title AnimStateConduitNode}
   AnimStateNode_0 = Wait_1() {is=title AnimStateNode}
   AnimStateNode_1 = Spinning 1() {is=title AnimStateNode}
   AnimStateNode_2 = Spinning 2() {is=title AnimStateNode}
   AnimStateTransitionNode_10 = AnimStateTransitionNode()
   AnimStateTransitionNode_11 = AnimStateTransitionNode() {automaticRuleTriggerTime=0, bAutomaticRuleBasedOnSequencePlayerInState=true}
   AnimStateTransitionNode_12 = AnimStateTransitionNode() {automaticRuleTriggerTime=0, crossfadeDuration=0}
   AnimStateTransitionNode_13 = AnimStateTransitionNode() {crossfadeDuration=0}
   AnimStateTransitionNode_7 = AnimStateTransitionNode() {automaticRuleTriggerTime=0, bAutomaticRuleBasedOnSequencePlayerInState=true}
   AnimStateTransitionNode_8 = AnimStateTransitionNode() {crossfadeDuration=0}
   AnimStateTransitionNode_9 = AnimStateTransitionNode() {automaticRuleTriggerTime=0, crossfadeDuration=0}
  ]
  links [
   {
    dst=Spinning 1
    rule=AnimGraph.AnimGraphNode_StateMachine_0.ChainSpinning SM.AnimStateTransitionNode_13.Transition
    src=IsSpinning? 1
    via=AnimStateTransitionNode_13
   }
   {
    dst=Wait_1
    rule=AnimGraph.AnimGraphNode_StateMachine_0.ChainSpinning SM.AnimStateTransitionNode_12.Transition
    src=IsSpinning? 1
    via=AnimStateTransitionNode_12
   }
   {
    dst=Spinning 2
    rule=AnimGraph.AnimGraphNode_StateMachine_0.ChainSpinning SM.AnimStateTransitionNode_8.Transition
    src=IsSpinning? 2
    via=AnimStateTransitionNode_8
   }
   {
    dst=Wait_1
    rule=AnimGraph.AnimGraphNode_StateMachine_0.ChainSpinning SM.AnimStateTransitionNode_9.Transition
    src=IsSpinning? 2
    via=AnimStateTransitionNode_9
   }
   {
    dst=IsSpinning? 2
    rule=AnimGraph.AnimGraphNode_StateMachine_0.ChainSpinning SM.AnimStateTransitionNode_7.Transition
    src=Spinning 1
    via=AnimStateTransitionNode_7
   }
   {
    dst=IsSpinning? 1
    rule=AnimGraph.AnimGraphNode_StateMachine_0.ChainSpinning SM.AnimStateTransitionNode_11.Transition
    src=Spinning 2
    via=AnimStateTransitionNode_11
   }
   {
    dst=IsSpinning? 2
    rule=AnimGraph.AnimGraphNode_StateMachine_0.ChainSpinning SM.AnimStateTransitionNode_10.Transition
    src=Wait_1
    via=AnimStateTransitionNode_10
   }
  ]
 }
 IsSpinning? 1 {path=AnimGraph.AnimGraphNode_StateMachine_0.ChainSpinning SM.AnimStateConduitNode_0.IsSpinning? 1, nodes [Result(bCanEnterTransition=true)]}
 IsSpinning? 2 {path=AnimGraph.AnimGraphNode_StateMachine_0.ChainSpinning SM.AnimStateConduitNode_1.IsSpinning? 2, nodes [Result(bCanEnterTransition=true)]}
 Wait_1 {
  path=AnimGraph.AnimGraphNode_StateMachine_0.ChainSpinning SM.AnimStateNode_0.Wait_1
  nodes [Output Animation Pose(Result=Sequence Player 'AS_Chainsaw_idle_Pose_Chainsaw_Sharp'() {bLoopAnimation=false, sequence=/ChainsawGameplay/Assets/Animations/ChainsawPropAnims/AS_Chainsaw_idle_Pose_Chainsaw_Sharp})]
 }
 Spinning 1 {
  path=AnimGraph.AnimGraphNode_StateMachine_0.ChainSpinning SM.AnimStateNode_1.Spinning 1
  nodes [Output Animation Pose(Result=Sequence Player 'AS_Chainsaw_idle_Noise_Chainsaw_Sharp'!(PlayRate:float=1) -> (Pose) {bLoopAnimation=false, sequence=/ChainsawGameplay/Assets/Animations/ChainsawPropAnims/AS_Chainsaw_idle_Noise_Chainsaw_Sharp})]
 }
 Spinning 2 {
  path=AnimGraph.AnimGraphNode_StateMachine_0.ChainSpinning SM.AnimStateNode_2.Spinning 2
  nodes [Output Animation Pose(Result=Sequence Player 'AS_Chainsaw_idle_Noise_Chainsaw_Sharp'(PlayRate=1) {bLoopAnimation=false, sequence=&AS_Chainsaw_idle_Noise_Chainsaw_Sharp})]
 }
 Transition {
  path=AnimGraph.AnimGraphNode_StateMachine_0.ChainSpinning SM.AnimStateTransitionNode_10.Transition
  nodes [Result(bCanEnterTransition=Get bIsSpinning() -> (bIsSpinning:bool))]
 }
 Transition {path=AnimGraph.AnimGraphNode_StateMachine_0.ChainSpinning SM.AnimStateTransitionNode_11.Transition, nodes [Result()]}
 Transition {
  path=AnimGraph.AnimGraphNode_StateMachine_0.ChainSpinning SM.AnimStateTransitionNode_12.Transition
  nodes [Result(bCanEnterTransition=NOT Boolean(A=Get bIsSpinning()))]
 }
 Transition {
  path=AnimGraph.AnimGraphNode_StateMachine_0.ChainSpinning SM.AnimStateTransitionNode_13.Transition
  nodes [Result(bCanEnterTransition=Get bIsSpinning())]
 }
 Transition {path=AnimGraph.AnimGraphNode_StateMachine_0.ChainSpinning SM.AnimStateTransitionNode_7.Transition, nodes [Result()]}
 Transition {
  path=AnimGraph.AnimGraphNode_StateMachine_0.ChainSpinning SM.AnimStateTransitionNode_8.Transition
  nodes [Result(bCanEnterTransition=Get bIsSpinning())]
 }
 Transition {
  path=AnimGraph.AnimGraphNode_StateMachine_0.ChainSpinning SM.AnimStateTransitionNode_9.Transition
  nodes [Result(bCanEnterTransition=NOT Boolean(A=Get bIsSpinning()))]
 }
 Equip SM {
  entry=Equipping
  path=AnimGraph.AnimGraphNode_StateMachine_1.Equip SM
  ids=id
  nodes [
   To Equip() {is=title AnimStateAliasNode, bGlobalAlias=true}
   To Done() {is=title AnimStateAliasNode, bGlobalAlias=true}
   AnimStateNode_0 = Equipping() {is=title AnimStateNode}
   AnimStateNode_1 = Done() {is=title AnimStateNode}
   AnimStateTransitionNode_4 = AnimStateTransitionNode() {bAutomaticRuleBasedOnSequencePlayerInState=true, crossfadeDuration=0}
   AnimStateTransitionNode_5 = AnimStateTransitionNode() {crossfadeDuration=0}
   AnimStateTransitionNode_6 = AnimStateTransitionNode() {crossfadeDuration=0}
  ]
  links [
   {
    dst=Done
    rule=AnimGraph.AnimGraphNode_StateMachine_1.Equip SM.AnimStateTransitionNode_4.Transition
    src=Equipping
    via=AnimStateTransitionNode_4
   }
   {
    dst=Done
    rule=AnimGraph.AnimGraphNode_StateMachine_1.Equip SM.AnimStateTransitionNode_6.Transition
    src=To Done
    via=AnimStateTransitionNode_6
   }
   {
    dst=Equipping
    rule=AnimGraph.AnimGraphNode_StateMachine_1.Equip SM.AnimStateTransitionNode_5.Transition
    src=To Equip
    via=AnimStateTransitionNode_5
   }
  ]
 }
 Equipping {
  path=AnimGraph.AnimGraphNode_StateMachine_1.Equip SM.AnimStateNode_0.Equipping
  nodes [
   n1 = Sequence Player 'AS_Chainsaw_run_equip_Chainsaw_Sharp'() {
    bLoopAnimation=false
    groupName=equipwep
    method=SyncGroup
    sequence=/ChainsawGameplay/Assets/Animations/ChainsawPropAnims/AS_Chainsaw_run_equip_Chainsaw_Sharp
   }
   n2 = Sequence Player 'AS_ChainSaw_Idle_Equip_Chainsaw_Sharp'() {
    bLoopAnimation=false
    groupName=equipwep
    method=SyncGroup
    sequence=/ChainsawGameplay/Assets/Animations/ChainsawPropAnims/AS_ChainSaw_Idle_Equip_Chainsaw_Sharp
   }
   n3 = Blend Poses by bool(BlendPose_0=@n1, BlendPose_1=@n2, bActiveValue=Get bIsSprinting()) {blendType=Linear}
   n4 = Sequence Player 'AS_ChainSaw_Slide_Equip_Chainsaw_Sharp'() {
    bLoopAnimation=false
    groupName=equipwep
    method=SyncGroup
    sequence=/ChainsawGameplay/Assets/Animations/ChainsawPropAnims/AS_ChainSaw_Slide_Equip_Chainsaw_Sharp
   }
   n5 = Get bIsFalling()
   n6 = Sequence Player 'AS_ChainSaw_Slide_Equip_Chainsaw_Sharp'() {
    bLoopAnimation=false
    groupName=equipwep
    method=SyncGroup
    sequence=&AS_ChainSaw_Slide_Equip_Chainsaw_Sharp
   }
   Output Animation Pose(Result=Blend Poses by bool(
     BlendPose_0=@n6
     BlendPose_1=Blend Poses by bool(BlendPose_0=@n4, BlendPose_1=@n3, bActiveValue=@n5) {blendType=Linear}
     bActiveValue=Get bIsSliding()
    ) {blendType=Linear})
  ]
 }
 Done {path=AnimGraph.AnimGraphNode_StateMachine_1.Equip SM.AnimStateNode_1.Done, nodes [Output Animation Pose(Result=Local Space Ref Pose())]}
 Transition {path=AnimGraph.AnimGraphNode_StateMachine_1.Equip SM.AnimStateTransitionNode_4.Transition, nodes [Result()]}
 Transition {
  path=AnimGraph.AnimGraphNode_StateMachine_1.Equip SM.AnimStateTransitionNode_5.Transition
  nodes [Result(bCanEnterTransition=Get bIsEquip() -> (bIsEquip:bool))]
 }
 Transition {
  path=AnimGraph.AnimGraphNode_StateMachine_1.Equip SM.AnimStateTransitionNode_6.Transition
  nodes [Result(bCanEnterTransition=Get bIsNotEquip() -> (bIsNotEquip:bool))]
 }
 EventGraph {links [], nodes []}
]
variables [
 As B Weapon Chainsaw {type=B_Weapon_Chainsaw_C, default=null}
 As Fort Player Pawn {type=FortPlayerPawn}
]
