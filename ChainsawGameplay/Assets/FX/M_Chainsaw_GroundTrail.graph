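# Unlit translucent Niagara ribbon material for the chainsaw's ground trail, tinted and faded by particle color.
# A panning trail texture, sampled straight and with noise-distorted V, forms the mask, which is faded at both U ends.
# A HeightLerp against cell noise, driven by the Dissolve dynamic parameter, erodes it before particle alpha yields the Opacity reroute.
# Particle RGB feeds an emissive-adjustment function; further branches compute UseFrontFade, noise-breakup and sine edge-shaping variants.
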Material /ChainsawGameplay/Assets/FX/M_Chainsaw_GroundTrail
graph {
 properties {bUsedWithNiagaraRibbons=true, blendMode=BLEND_Translucent, pixelDepthOffsetMode=PDOM_Legacy, shadingModel=MSM_Unlit, usedShadingModels=Unlit}
 nodes [
  n1 = ParticleColor()
  n2 = VectorParameter() {parameterName=UVScale, defaultValue=<rgba 0.4 1 0 0>}
  n3 = Panner(Coordinate=Multiply(A=TextureCoordinate(), B=ComponentMask(None=@n2.RGB) {g=true, r=true})) {speedX=-1.5}
  n4 = TextureSampleParameter2D(UVs=@n3) {parameterName=Tex, texture=&/T_FX_Chainsaw_GroundTrail_Rough}
  n5 = Multiply(A=@n4.R, B=LinearInterpolate(Alpha=OneMinus(None=@n4.G)) {constA=0.6})
  n6 = Constant()
  n7 = TextureCoordinate()
  n8 = LinearInterpolate(Alpha=TextureSample(UVs=Panner(Coordinate=@n7) {speedX=-2}) {samplerType=SAMPLERTYPE_Grayscale, texture=&LowResBlurredNoise}.R) {constA=-1}
  n9 = Add(A=ComponentMask(None=Multiply(A=@n8) {constB=0.3}) {r=true}) {constB=0.05}
  n10 = TextureSampleParameter2D(UVs=Add(A=@n3, B=AppendVector(A=@n6, B=@n9)) {constB=0}) {parameterName=Tex, texture=&T_FX_Chainsaw_GroundTrail_Rough}
  n11 = LinearInterpolate(Alpha=Saturate(None=Multiply(A=@n10.B) {constB=10})) {constA=0.1}
  n12 = Power(Base=ComponentMask(None=TextureCoordinate()) {r=true})
  n13 = Multiply(
   A=StaticComponentMaskParameter(None=Multiply(A=@n5, B=@n11)) {defaultR=true, parameterName=Mask}
   B=Saturate(None=Multiply(A=@n12) {constB=20})
  )
  n14 = TextureCoordinate()
  n15 = Power(Base=OneMinus(None=ComponentMask(None=@n14) {r=true}))
  n16 = Multiply(A=@n13, B=Saturate(None=Multiply(A=@n15) {constB=5}))
  n17 = Panner(Coordinate=TextureCoordinate() {uTiling=0.5, vTiling=0.7})
  n18 = Multiply(
   A=@n16
   B=MaterialFunctionCall(
    Transition Phase=DynamicParameter() {paramNames [Dissolve,Param2,Param3,Param4]}.Dissolve
    Height Texture=TextureSample(UVs=@n17) {samplerType=SAMPLERTYPE_Grayscale, texture=/Game/Weapons/FORT_Melee/Pickaxe_Cube_Ninja_Male/FX/T_CellNoise}.R
   ) {materialFunction=&HeightLerp}.Alpha
  ) {constA=1}
  n19 = Multiply(A=@n1.A, B=@n18)
  n20 = Power(Base=ComponentMask(None=TextureCoordinate()) {r=true}) {constExponent=1}
  LinearInterpolate(A=@n19, B=Multiply(A=@n19, B=@n20), Alpha=ScalarParameter() {parameterName=UseFrontFade})
  LinearInterpolate(A=@n13, B=@n16, Alpha=ScalarParameter() {parameterName=UseFrontFade})
  n21 = TextureSample(UVs=TextureCoordinate() {uTiling=0}) {samplerType=SAMPLERTYPE_Masks, texture=/Game/Weapons/WeaponSkins/Wraps/Textures/Mash/Mash_Noise}
  LinearInterpolate(A=@n18, B=Multiply(A=@n18, B=LinearInterpolate(Alpha=@n21.R) {constA=-0.2, constB=2}))
  MaterialFunctionCall(Emissive=@n1.RGB) {materialFunction=&MF_FX_EmissiveAdjustments}
  NamedRerouteDeclaration(None=@n19) {name=Opacity, nodeColor=<rgba 0 1 0.9063 1>}
  NamedRerouteUsage()
  n22 = Sine(None=ComponentMask(None=TextureCoordinate()) {g=true}) {period=2}
  n23 = Subtract(A=@n22)
  Saturate(None=Multiply(A=LinearInterpolate(A=@n22, B=@n23, Alpha=@n20) {constA=1, constB=100}) {constB=5})
 ]
}
