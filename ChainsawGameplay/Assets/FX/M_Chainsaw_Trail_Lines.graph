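# Unlit, two-sided translucent material for Niagara ribbons that draws streaky chainsaw trail lines.
# Cloud noise is sampled using only the ribbon's V coordinate, scaled and offset by dynamic parameters, so it forms lines along U, sharpened by a power.
# Fades at both U ends and particle alpha mask the lines.
# A feature-level switch adds adjustable depth fade, skipping it on ES2 and ES3.1.
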Material /ChainsawGameplay/Assets/FX/M_Chainsaw_Trail_Lines
graph {
 properties {
  bAutomaticallySetUsageInEditor=false
  bUsedWithNiagaraRibbons=true
  blendMode=BLEND_Translucent
  shadingModel=MSM_Unlit
  twoSided=true
  usedShadingModels=Unlit
 }
 nodes [
  n1 = DynamicParameter() {paramNames [U,VMult,Param3,Param4]}
  n2 = MaterialFunctionCall(Float2=TextureCoordinate()) {materialFunction=&BreakOutFloat2Components}
  n3 = TextureSample(UVs=Add(A=@n1.U, B=Multiply(A=@n2.G, B=@n1.VMult))) {texture=&AtmosphericCloudNoise01}
  n4 = MaterialFunctionCall(Float2=TextureCoordinate()) {materialFunction=&BreakOutFloat2Components}
  n5 = Power(Base=OneMinus(None=@n4.R))
  n6 = Saturate(None=Power(Base=Multiply(A=@n4.R) {constB=5}))
  n7 = Multiply(A=ParticleColor().A, B=Multiply(A=Power(Base=@n3.RGB) {constExponent=3}, B=Multiply(A=@n5, B=@n6)))
  n8 = DepthFade(Opacity=@n7, FadeDistance=ScalarParameter() {defaultValue=30, parameterName=Depth Fade Distance, sliderMax=200})
  FeatureLevelSwitch(Default=@n8, ES2=@n7, ES3_1=@n7, SM4_REMOVED=@n8, SM5=@n8, SM6=@n8)
 ]
}
