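# Niagara effect for a chainsaw grinding along the ground, throwing sparks, dirt, smoke and debris.
# GPU sparks stretch with speed and grow with LOD distance, beside a glow flare whose light flickers on a noisy curve.
# Sprite and mesh dirt, smoke, speed lines and physical-surface debris layer around them, most inheriting source movement.
# A distance-spawned ground ribbon locks its head to the simulation position and, like the smoke spikes, tints by time of day.
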NiagaraSystem /ChainsawGameplay/Assets/FX/NS_Chainsaw_GroundSparks
dependencies {
 usedDataInterfaces [&NiagaraDataInterfaceCurve, &NiagaraDataInterfaceColorCurve, &NiagaraDataInterfaceVector2DCurve]
 usedDynamicInputs [&SimulationPosition, &Vector2DFromCurve]
 usedModules [&SystemState]
}
emitters {
 Dirt_Chunks {
  emitterUpdateScript [
   &EmitterState(Life Cycle Mode=::NewEnumerator1, Loop Behavior=::NewEnumerator1, Loop Delay=0.1, Loop Duration Mode=::NewEnumerator1, UseLoopDelay=true)
   &SpawnRate(Spawn Probability=0.67, SpawnRate=30, Use Spawn Probability=false)
  ]
  particleSpawnScript [
   /SetVariables(Particles.Rand:NiagaraFloat=&UniformAOrBFloat(A=0, Evaluation Type=::NewEnumerator0, Recalculate Random Each Loop=true, Override Seed=false)) {isSetParameters=true, module=SetVariables_#1}
   &InitializeParticle(
    Color=<rgba 0.0469 0.035 0.0308 2>
    Lifetime=1.5
    Position Offset=&Lerp_Vector(A=<-10 0 0>, B=<10 0 0>, Alpha=$Particles.Rand)
    Sprite Rotation Mode=::NewEnumerator2
    Sprite Size=<300 300>
    Sprite Size Mode=::NewEnumerator4
    UsePositionOffset=true
    Write Material Random=true
   )
   &CameraOffset(Camera Offset Amount=20)
   &AddVelocity(
    Cone Angle=50
    Cone Axis=&Lerp_Vector(A=<-1 1 -0.5>, B=<1 1 -0.25>, Alpha=$Particles.Rand)
    Module.UseConeFalloff=false
    Module.UseVelDistribution=false
    Override Seed=false
    Velocity Mode=::NewEnumerator2
    Velocity Speed=&Lerp_Float(
     A=0.15
     B=0.25
     Alpha=&FloatFromCurve(
      FloatCurve=DataInterface() {propertyValues {Curve {keys [
          {interpMode=RCIM_Cubic, tangentMode=RCTM_User, time=0, value=0}
          {interpMode=RCIM_Cubic, tangentMode=RCTM_User, time=0.5, value=1}
          {tangentMode=RCTM_User, time=1, value=0}
         ]}}}
      CurveIndex=&RandomRangeFloat(Minimum=0, Maximum=1)
      Scale Curve=3
     )
    )
   ) {module=AddVelocity003}
  ]
  particleUpdateScript [
   &ParticleState()
   &DynamicMaterialParameters(
    Index 0 Param 1=$Particles.NormalizedAge
    Index 0 Param 2=&RandomRangeFloat(Minimum=1, Maximum=1.5)
    Index 0 Param 3=&RandomRangeFloat(Minimum=0, Maximum=1)
    Index 0 Param 4=&FloatFromCurve(
     FloatCurve=DataInterface() {propertyValues {Curve {keys [
         {arriveTangent=2.9585, interpMode=RCIM_Cubic, leaveTangent=2.9585, tangentMode=RCTM_User, time=0, value=0}
         {arriveTangent=-0.0405, interpMode=RCIM_Cubic, leaveTangent=-0.0405, tangentMode=RCTM_User, time=1, value=1}
        ]}}}
     Scale Curve=&RandomRangeFloat(Minimum=2, Maximum=3)
    )
    Module.Param0WriteEnabled=true
    Module.Param1WriteEnabled=true
    Module.Param2WriteEnabled=true
    Module.Param3WriteEnabled=true
   )
   &Drag(Drag=4, Use Linear Drag=true, Use Rotational Drag=false)
   &ScaleColor(
    Scale Alpha=&FloatFromCurve(
     FloatCurve=DataInterface() {propertyValues {Curve {keys [{interpMode=RCIM_Cubic, tangentMode=RCTM_User, time=0.1, value=1}, {interpMode=RCIM_Cubic, tangentMode=RCTM_User, time=1, value=0}]}}}
     Scale Curve=0.5
    )
    Scale RGB=&VectorFromFloat(Value=&FloatFromCurve(FloatCurve=DataInterface() {propertyValues {Curve {keys [{arriveTangent=-3.4125, interpMode=RCIM_Cubic, leaveTangent=-3.4125, tangentMode=RCTM_User, time=0, value=3}, <at 0.35 2>]}}}))
    ScaleA=true
    ScaleRGB=true
   )
   &InheritSourceMovement(
    Applied Position Scale=&Multiply_VectorByFloat(
     Vector=<1 1 1>
     Float=&FloatFromCurve(FloatCurve=DataInterface() {propertyValues {Curve {keys [{interpMode=RCIM_Cubic, tangentMode=RCTM_User, time=0, value=1}, {interpMode=RCIM_Cubic, tangentMode=RCTM_User, time=1, value=0}]}}})
    )
    ApplyForce=false
    ApplyPosition=true
    ApplyPositionFalloff=false
   )
   &SolveForcesAndVelocity(Clamp Velocity=false, Limit Acceleration=false)
   &ScaleSpriteSize(
    Uniform Curve Scale=0.45
    Uniform Curve Sprite Scale=DataInterface() {propertyValues {Curve {keys [
        {arriveTangent=3.844, interpMode=RCIM_Cubic, leaveTangent=3.844, tangentMode=RCTM_User, time=0, value=0}
        {arriveTangent=0.1649, interpMode=RCIM_Cubic, leaveTangent=0.1649, tangentMode=RCTM_User, time=0.75, value=1}
        <at 1 0 RCIM_Cubic>
       ]}}}
   )
   &SpriteRotationRate(Rotation Rate=&RandomRangeFloat(Minimum=-45, Maximum=45))
  ]
  rendererClasses [&NiagaraSpriteRendererProperties]
  renderers [{rendererClass=&NiagaraSpriteRendererProperties}]
 }
 Dirts {
  emitterUpdateScript [
   &EmitterState(Life Cycle Mode=::NewEnumerator1, Loop Behavior=::NewEnumerator1, Loop Delay=0.6, Loop Duration Mode=::NewEnumerator1, UseLoopDelay=false)
   &SpawnRate(Spawn Probability=0.67, SpawnRate=180, Use Spawn Probability=true)
  ]
  particleSpawnScript [
   &InitializeParticle(
    Color=<rgba 0.6 0.4 0.2 0.78>
    Lifetime Max=0.5
    Lifetime Min=0.3
    Lifetime Mode=::NewEnumerator1
    Sprite Rotation Mode=::NewEnumerator2
    Sprite Size Mode=::NewEnumerator1
    Uniform Sprite Size Max=75
    Uniform Sprite Size Min=60
    UsePositionOffset=false
    Write Material Random=true
   ) {module=InitializeParticle001}
   &AddVelocity(
    Cone Angle=20
    Cone Axis=&Lerp_Vector(
     A=<1 1 -0.5>
     B=<-1 1 -0.5>
     Alpha=&UniformAOrBFloat(A=0, Evaluation Type=::NewEnumerator0, Recalculate Random Each Loop=true, Override Seed=false)
    )
    Inner Cone Angle=5
    Module.UseConeFalloff=false
    Module.UseVelDistribution=false
    Override Seed=false
    Velocity Mode=::NewEnumerator2
    Velocity Speed=&Lerp_Float(
     A=150
     B=250
     Alpha=&FloatFromCurve(
      FloatCurve=DataInterface() {propertyValues {Curve {keys [
          {interpMode=RCIM_Cubic, tangentMode=RCTM_User, time=0, value=0}
          {interpMode=RCIM_Cubic, tangentMode=RCTM_User, time=0.5, value=1}
          {tangentMode=RCTM_User, time=1, value=0}
         ]}}}
      CurveIndex=&RandomRangeFloat(Minimum=0, Maximum=1)
      Scale Curve=3
     )
    )
   ) {module=AddVelocity003}
  ]
  particleUpdateScript [
   &ParticleState()
   &GravityForce(Gravity=<0 0 -600>)
   &Drag(Drag=&RandomRangeFloat(Minimum=0.75, Maximum=1), Use Linear Drag=true, Use Rotational Drag=false)
   &ScaleColor(
    Scale Alpha=&FloatFromCurve(
     FloatCurve=DataInterface() {propertyValues {Curve {keys [
         {arriveTangent=0.0639, interpMode=RCIM_Cubic, leaveTangent=0.0639, tangentMode=RCTM_User, time=0, value=0}
         {arriveTangent=0.0047, interpMode=RCIM_Cubic, leaveTangent=0.0047, tangentMode=RCTM_User, time=0.1408, value=0.9847}
         {arriveTangent=-1.146, interpMode=RCIM_Cubic, leaveTangent=-1.146, time=0.8, value=0.8}
         {arriveTangent=-0.0405, interpMode=RCIM_Cubic, leaveTangent=-0.0405, tangentMode=RCTM_User, time=1, value=0}
        ]}}}
     Scale Curve=1.03
    )
    Scale RGB=<0.3 0.3 0.3>
    ScaleA=true
    ScaleRGB=true
   )
   &ScaleSpriteSize(Scale Factor=&Vector2DFromFloat(Value=&FloatFromCurve(
      FloatCurve=DataInterface() {propertyValues {Curve {keys [
          {arriveTangent=1.5713, interpMode=RCIM_Cubic, leaveTangent=1.5713, tangentMode=RCTM_User, time=0, value=0.1}
          {arriveTangent=0.0879, interpMode=RCIM_Cubic, leaveTangent=0.0879, tangentMode=RCTM_User, time=1, value=0.8}
         ]}}}
      Scale Curve=2.2
     )))
   &InheritSourceMovement(ApplyForce=false, ApplyPosition=true, ApplyPositionFalloff=false)
   &SolveForcesAndVelocity(Clamp Velocity=false, Limit Acceleration=false)
  ]
  rendererClasses [&NiagaraSpriteRendererProperties]
  renderers [{rendererClass=&NiagaraSpriteRendererProperties}]
 }
 Glow_Flare {
  emitterUpdateScript [
   &EmitterState(Life Cycle Mode=::NewEnumerator1, Loop Behavior=::NewEnumerator1, Loop Delay=0.6, Loop Duration Mode=::NewEnumerator1, UseLoopDelay=false)
   &SpawnBurst_Instantaneous(Spawn Count=1, Use Spawn Probability=false)
  ]
  particleSpawnScript [
   &InitializeParticle(
    Color=<rgba 1 0.3963 0 1>
    Lifetime=0.1
    Mass Mode=::NewEnumerator1
    Sprite Size=<65 65>
    Sprite Size Mode=::NewEnumerator4
    UsePositionOffset=false
    Write Material Random=true
   ) {module=InitializeParticle001}
   /SetVariables(Particles.Color_Glow:LinearColor=<rgba 0.125 0.0495 0 0.7>, Particles.DMP_Glow:Vector4f=<1 1 1 1>) {isSetParameters=true}
   &CameraOffset(Camera Offset Amount=10)
   &DynamicMaterialParameters(
    Index 0 Param 1=0.2
    Module.Param0WriteEnabled=true
    Module.Param1WriteEnabled=true
    Module.Param2WriteEnabled=true
    Module.Param3WriteEnabled=true
   )
  ]
  particleUpdateScript [
   &ParticleState(Kill Particles When Lifetime Has Elapsed=false, Let Infinitely Lived Particles Die When Emitter Deactivates=true)
   &ScaleColor(
    Scale Alpha=&Multiply_Float(
     A=&RemapRange(
      Input Value To Remap=&FloatFromCurve(
       FloatCurve=DataInterface() {propertyValues {
         bUseLUT=false
         Curve {
          postInfinityExtrap=RCCE_Cycle
          preInfinityExtrap=RCCE_Cycle
          keys [
           <at 0 1 RCIM_Cubic>
           {arriveTangent=0.5283, interpMode=RCIM_Cubic, leaveTangent=0.5283, time=0.0746, value=0.875}
           {arriveTangent=-0.0227, interpMode=RCIM_Cubic, leaveTangent=-0.0227, tangentMode=RCTM_User, time=0.1682, value=1.0888}
           {arriveTangent=-1.3938, interpMode=RCIM_Cubic, leaveTangent=-1.3938, time=0.2019, value=0.6577}
           {arriveTangent=1.4729, interpMode=RCIM_Cubic, leaveTangent=1.4729, time=0.2748, value=0.9402}
           {arriveTangent=0.905, interpMode=RCIM_Cubic, leaveTangent=0.905, time=0.3464, value=0.8705}
           {arriveTangent=-1.2159, interpMode=RCIM_Cubic, leaveTangent=-1.2159, time=0.4074, value=1.0603}
           {arriveTangent=-1.5422, interpMode=RCIM_Cubic, leaveTangent=-1.5422, time=0.4365, value=0.7609}
           {arriveTangent=-0.8529, interpMode=RCIM_Cubic, leaveTangent=-0.8529, time=0.4871, value=0.9373}
           {arriveTangent=0.3027, interpMode=RCIM_Cubic, leaveTangent=0.3027, time=0.5331, value=0.6785}
           {arriveTangent=1.4034, interpMode=RCIM_Cubic, leaveTangent=1.4034, time=0.6112, value=0.9748}
           {arriveTangent=1.5913, interpMode=RCIM_Cubic, leaveTangent=1.5913, time=0.69, value=0.8986}
           {arriveTangent=-0.751, interpMode=RCIM_Cubic, leaveTangent=-0.751, time=0.7536, value=1.2014}
           {arriveTangent=-2.595, interpMode=RCIM_Cubic, leaveTangent=-2.595, time=0.7995, value=0.8164}
           {arriveTangent=-0.8671, interpMode=RCIM_Cubic, leaveTangent=-0.8671, time=0.8674, value=0.9059}
           {arriveTangent=-0.3671, interpMode=RCIM_Cubic, leaveTangent=-0.3671, time=0.9138, value=0.7173}
           {arriveTangent=-0.6714, interpMode=RCIM_Cubic, leaveTangent=-0.6714, time=0.9899, value=0.8609}
           {arriveTangent=0.7195, interpMode=RCIM_Cubic, leaveTangent=0.7195, time=1.0372, value=0.6345}
           {arriveTangent=1.5649, interpMode=RCIM_Cubic, leaveTangent=1.5649, time=1.099, value=0.9394}
           {arriveTangent=0.3682, interpMode=RCIM_Cubic, leaveTangent=0.3682, time=1.1296, value=0.7792}
           {arriveTangent=1.199, interpMode=RCIM_Cubic, leaveTangent=1.199, time=1.2078, value=0.9795}
           {arriveTangent=0.6999, interpMode=RCIM_Cubic, leaveTangent=0.6999, time=1.2281, value=0.8972}
           {arriveTangent=0.8013, interpMode=RCIM_Cubic, leaveTangent=0.8013, time=1.329, value=1.0643}
           {arriveTangent=0.4649, interpMode=RCIM_Cubic, leaveTangent=0.4649, time=1.3719, value=1.0125}
           {arriveTangent=-3.0415, interpMode=RCIM_Cubic, leaveTangent=-3.0415, time=1.414, value=1.1038}
           {arriveTangent=-0.0928, interpMode=RCIM_Cubic, leaveTangent=-0.0928, tangentMode=RCTM_User, time=1.4501, value=0.7746}
           {arriveTangent=0.7682, interpMode=RCIM_Cubic, leaveTangent=0.7682, time=1.5242, value=0.9899}
           {arriveTangent=0.5284, interpMode=RCIM_Cubic, leaveTangent=0.5284, time=1.5566, value=0.8565}
           {arriveTangent=-0.1056, interpMode=RCIM_Cubic, leaveTangent=-0.1056, time=1.7244, value=1.0957}
           {arriveTangent=-2.1193, interpMode=RCIM_Cubic, leaveTangent=-2.1193, time=1.7619, value=0.8348}
           {arriveTangent=-1.2854, interpMode=RCIM_Cubic, leaveTangent=-1.2854, time=1.7959, value=0.9442}
           {arriveTangent=-0.1279, interpMode=RCIM_Cubic, leaveTangent=-0.1279, time=1.836, value=0.7395}
           {arriveTangent=1.0187, interpMode=RCIM_Cubic, leaveTangent=1.0187, time=1.9343, value=0.9265}
           {arriveTangent=1.0782, interpMode=RCIM_Cubic, leaveTangent=1.0782, time=1.9912, value=0.8977}
           {arriveTangent=-2.5847, interpMode=RCIM_Cubic, leaveTangent=-2.5847, time=2.0488, value=1.0499}
           {arriveTangent=-0.6379, interpMode=RCIM_Cubic, leaveTangent=-0.6379, time=2.0828, value=0.6611}
           {arriveTangent=0.4353, interpMode=RCIM_Cubic, leaveTangent=0.4353, time=2.2158, value=0.9434}
           {arriveTangent=-0.3447, interpMode=RCIM_Cubic, leaveTangent=-0.3447, time=2.2673, value=0.7414}
           {arriveTangent=0.6603, interpMode=RCIM_Cubic, leaveTangent=0.6603, time=2.3258, value=0.9055}
           {arriveTangent=0.3801, interpMode=RCIM_Cubic, leaveTangent=0.3801, time=2.3971, value=0.8271}
           {arriveTangent=0.6687, interpMode=RCIM_Cubic, leaveTangent=0.6687, time=2.4654, value=0.9586}
           {arriveTangent=1.2527, interpMode=RCIM_Cubic, leaveTangent=1.2527, time=2.5143, value=0.9055}
           {arriveTangent=-0.515, interpMode=RCIM_Cubic, leaveTangent=-0.515, time=2.5566, value=1.0728}
           {arriveTangent=-1.6201, interpMode=RCIM_Cubic, leaveTangent=-1.6201, time=2.597, value=0.8629}
           {arriveTangent=-0.9669, interpMode=RCIM_Cubic, leaveTangent=-0.9669, time=2.6456, value=0.9286}
           {arriveTangent=0.1009, interpMode=RCIM_Cubic, leaveTangent=0.1009, time=2.6921, value=0.771}
           {arriveTangent=0.0929, interpMode=RCIM_Cubic, leaveTangent=0.0929, time=2.7509, value=0.9392}
           {arriveTangent=0.2939, interpMode=RCIM_Cubic, leaveTangent=0.2939, time=2.8654, value=0.7871}
           {arriveTangent=1.7227, interpMode=RCIM_Cubic, leaveTangent=1.7227, time=2.9271, value=0.991}
           {arriveTangent=0.1236, interpMode=RCIM_Cubic, leaveTangent=0.1236, time=2.957, value=0.9449}
           <at 3 1 RCIM_Cubic>
          ]
         }
        }}
       CurveIndex=&Multiply_Float(A=$Particles.NormalizedAge, B=0.1)
      )
      Input Min=0.6
      Input Max=1.2
      Output Min=0.2
      Output Max=1
     )
     B=&FloatFromCurve(FloatCurve=DataInterface() {propertyValues {Curve {keys [
          {arriveTangent=17.213, interpMode=RCIM_Cubic, leaveTangent=17.213, tangentMode=RCTM_User, time=0, value=0}
          {arriveTangent=0.3861, interpMode=RCIM_Cubic, leaveTangent=0.3861, tangentMode=RCTM_User, time=0.1, value=0.9847}
         ]}}})
    )
    Scale RGB=&VectorFromFloat(Value=&Multiply_Float(A=50, B=1))
    ScaleA=true
    ScaleRGB=true
   )
   /SetVariables(Particles.SpriteSize_Glow:Vector2f=&Multiply_Vector2D_ByFloat(
     Float=&FloatFromCurve(FloatCurve=DataInterface() {propertyValues {Curve {keys [<at 0 1>,<at 1 3>]}}})
     Vector2D=<50 50>
    )) {isSetParameters=true}
   &FN_Light_v2(
    Light Exponent=2
    Light Radius=80
    LightEnabled=true
    Position Offset:Vector3f=<-15 0 0>
    Scale Color:NiagaraFloat=&Multiply_Float(
     A=1.5
     B=&RemapRange(
      Input Value To Remap=&FloatFromCurve(
       FloatCurve=DataInterface() {propertyValues {
         bUseLUT=false
         Curve {
          postInfinityExtrap=RCCE_Cycle
          preInfinityExtrap=RCCE_Cycle
          keys [
           <at 0 1 RCIM_Cubic>
           {arriveTangent=0.5283, interpMode=RCIM_Cubic, leaveTangent=0.5283, time=0.0746, value=0.875}
           {arriveTangent=-0.0227, interpMode=RCIM_Cubic, leaveTangent=-0.0227, tangentMode=RCTM_User, time=0.1682, value=1.0888}
           {arriveTangent=-1.3938, interpMode=RCIM_Cubic, leaveTangent=-1.3938, time=0.2019, value=0.6577}
           {arriveTangent=1.4729, interpMode=RCIM_Cubic, leaveTangent=1.4729, time=0.2748, value=0.9402}
           {arriveTangent=0.905, interpMode=RCIM_Cubic, leaveTangent=0.905, time=0.3464, value=0.8705}
           {arriveTangent=-1.2159, interpMode=RCIM_Cubic, leaveTangent=-1.2159, time=0.4074, value=1.0603}
           {arriveTangent=-1.5422, interpMode=RCIM_Cubic, leaveTangent=-1.5422, time=0.4365, value=0.7609}
           {arriveTangent=-0.8529, interpMode=RCIM_Cubic, leaveTangent=-0.8529, time=0.4871, value=0.9373}
           {arriveTangent=0.3027, interpMode=RCIM_Cubic, leaveTangent=0.3027, time=0.5331, value=0.6785}
           {arriveTangent=1.4034, interpMode=RCIM_Cubic, leaveTangent=1.4034, time=0.6112, value=0.9748}
           {arriveTangent=1.5913, interpMode=RCIM_Cubic, leaveTangent=1.5913, time=0.69, value=0.8986}
           {arriveTangent=-0.751, interpMode=RCIM_Cubic, leaveTangent=-0.751, time=0.7536, value=1.2014}
           {arriveTangent=-2.595, interpMode=RCIM_Cubic, leaveTangent=-2.595, time=0.7995, value=0.8164}
           {arriveTangent=-0.8671, interpMode=RCIM_Cubic, leaveTangent=-0.8671, time=0.8674, value=0.9059}
           {arriveTangent=-0.3671, interpMode=RCIM_Cubic, leaveTangent=-0.3671, time=0.9138, value=0.7173}
           {arriveTangent=-0.6714, interpMode=RCIM_Cubic, leaveTangent=-0.6714, time=0.9899, value=0.8609}
           {arriveTangent=0.7195, interpMode=RCIM_Cubic, leaveTangent=0.7195, time=1.0372, value=0.6345}
           {arriveTangent=1.5649, interpMode=RCIM_Cubic, leaveTangent=1.5649, time=1.099, value=0.9394}
           {arriveTangent=0.3682, interpMode=RCIM_Cubic, leaveTangent=0.3682, time=1.1296, value=0.7792}
           {arriveTangent=1.199, interpMode=RCIM_Cubic, leaveTangent=1.199, time=1.2078, value=0.9795}
           {arriveTangent=0.6999, interpMode=RCIM_Cubic, leaveTangent=0.6999, time=1.2281, value=0.8972}
           {arriveTangent=0.8013, interpMode=RCIM_Cubic, leaveTangent=0.8013, time=1.329, value=1.0643}
           {arriveTangent=0.4649, interpMode=RCIM_Cubic, leaveTangent=0.4649, time=1.3719, value=1.0125}
           {arriveTangent=-3.0415, interpMode=RCIM_Cubic, leaveTangent=-3.0415, time=1.414, value=1.1038}
           {arriveTangent=-0.0928, interpMode=RCIM_Cubic, leaveTangent=-0.0928, tangentMode=RCTM_User, time=1.4501, value=0.7746}
           {arriveTangent=0.7682, interpMode=RCIM_Cubic, leaveTangent=0.7682, time=1.5242, value=0.9899}
           {arriveTangent=0.5284, interpMode=RCIM_Cubic, leaveTangent=0.5284, time=1.5566, value=0.8565}
           {arriveTangent=-0.1056, interpMode=RCIM_Cubic, leaveTangent=-0.1056, time=1.7244, value=1.0957}
           {arriveTangent=-2.1193, interpMode=RCIM_Cubic, leaveTangent=-2.1193, time=1.7619, value=0.8348}
           {arriveTangent=-1.2854, interpMode=RCIM_Cubic, leaveTangent=-1.2854, time=1.7959, value=0.9442}
           {arriveTangent=-0.1279, interpMode=RCIM_Cubic, leaveTangent=-0.1279, time=1.836, value=0.7395}
           {arriveTangent=1.0187, interpMode=RCIM_Cubic, leaveTangent=1.0187, time=1.9343, value=0.9265}
           {arriveTangent=1.0782, interpMode=RCIM_Cubic, leaveTangent=1.0782, time=1.9912, value=0.8977}
           {arriveTangent=-2.5847, interpMode=RCIM_Cubic, leaveTangent=-2.5847, time=2.0488, value=1.0499}
           {arriveTangent=-0.6379, interpMode=RCIM_Cubic, leaveTangent=-0.6379, time=2.0828, value=0.6611}
           {arriveTangent=0.4353, interpMode=RCIM_Cubic, leaveTangent=0.4353, time=2.2158, value=0.9434}
           {arriveTangent=-0.3447, interpMode=RCIM_Cubic, leaveTangent=-0.3447, time=2.2673, value=0.7414}
           {arriveTangent=0.6603, interpMode=RCIM_Cubic, leaveTangent=0.6603, time=2.3258, value=0.9055}
           {arriveTangent=0.3801, interpMode=RCIM_Cubic, leaveTangent=0.3801, time=2.3971, value=0.8271}
           {arriveTangent=0.6687, interpMode=RCIM_Cubic, leaveTangent=0.6687, time=2.4654, value=0.9586}
           {arriveTangent=1.2527, interpMode=RCIM_Cubic, leaveTangent=1.2527, time=2.5143, value=0.9055}
           {arriveTangent=-0.515, interpMode=RCIM_Cubic, leaveTangent=-0.515, time=2.5566, value=1.0728}
           {arriveTangent=-1.6201, interpMode=RCIM_Cubic, leaveTangent=-1.6201, time=2.597, value=0.8629}
           {arriveTangent=-0.9669, interpMode=RCIM_Cubic, leaveTangent=-0.9669, time=2.6456, value=0.9286}
           {arriveTangent=0.1009, interpMode=RCIM_Cubic, leaveTangent=0.1009, time=2.6921, value=0.771}
           {arriveTangent=0.0929, interpMode=RCIM_Cubic, leaveTangent=0.0929, time=2.7509, value=0.9392}
           {arriveTangent=0.2939, interpMode=RCIM_Cubic, leaveTangent=0.2939, time=2.8654, value=0.7871}
           {arriveTangent=1.7227, interpMode=RCIM_Cubic, leaveTangent=1.7227, time=2.9271, value=0.991}
           {arriveTangent=0.1236, interpMode=RCIM_Cubic, leaveTangent=0.1236, time=2.957, value=0.9449}
           <at 3 1 RCIM_Cubic>
          ]
         }
        }}
       CurveIndex=&Multiply_Float(A=$Particles.NormalizedAge, B=0.12)
      )
      Input Min=0.6
      Input Max=1.2
      Output Min=0.2
      Output Max=1
     )
    )
   )
  ]
  rendererClasses [&NiagaraSpriteRendererProperties, &NiagaraLightRendererProperties]
  renderers [
   {rendererClass=&NiagaraSpriteRendererProperties}
   {rendererClass=&NiagaraSpriteRendererProperties, rendererIndex=1}
   {rendererClass=&NiagaraLightRendererProperties, rendererIndex=2}
  ]
 }
 PhysMatDebris {
  emitterUpdateScript [
   &EmitterState(Life Cycle Mode=::NewEnumerator1, Loop Behavior=::NewEnumerator1, Loop Delay=0.6, Loop Duration Mode=::NewEnumerator1, UseLoopDelay=false)
   &SpawnRate(Spawn Probability=0.67, SpawnRate=50, Use Spawn Probability=true)
   &SpawnBurst_Instantaneous(
    Spawn Count=&RandomRangeInt(Minimum=2, Maximum=4)
    Use Spawn Probability=false
    UseLoopCountLimit=false
   )
  ]
  particleSpawnScript [
   /SetVariables(Particles.BiasedRandom=&FloatFromCurve(
     FloatCurve=DataInterface() {propertyValues {Curve {keys [
         <at 0.2 1 RCIM_Cubic>
         {arriveTangent=-3.3394, interpMode=RCIM_Cubic, leaveTangent=-3.3394, tangentMode=RCTM_User, time=0.5, value=0.5}
         <at 0.8 0 RCIM_Cubic>
        ]}}}
     CurveIndex=&RandomRangeFloat(Minimum=0, Maximum=1)
    )) {isSetParameters=true}
   &InitializeParticle(
    Color Mode=::NewEnumerator0
    Lifetime=&RandomRangeFloat(Minimum=0.35, Maximum=0.7)
    Mass=&Lerp_Float(A=0.6, B=1.2, Alpha=$Particles.BiasedRandom)
    Mass Mode=::NewEnumerator1
    Mesh Scale=&Multiply_VectorByFloat(
     Vector=&RandomRangeVector(Minimum=<0.9 0.4 0.5>, Maximum=<1 1 1>)
     Float=&Lerp_Float(A=0.7, B=1, Alpha=$Particles.BiasedRandom)
    )
    Mesh Scale Mode=::NewEnumerator4
    Position Offset=<0 0 3>
    UsePositionOffset=false
    Write Material Random=true
   )
   &FN_PhysicalSurface(
    Dirt Mesh Count=2
    Glass Mesh Count=3
    Ice Mesh Count=2
    Metal Gravity=&RandomRangeFloat(Minimum=-350, Maximum=-1800)
    Metal Mesh Count=2
    Metal Velocity=&RandomRangeFloat(Minimum=80, Maximum=375)
    Metal Velocity Cone Angle=130
    Physical Surface=2
    Stone Gravity=&RandomRangeFloat(Minimum=-700, Maximum=-1250)
    Stone Mesh Count=2
    Stone Velocity=&RandomRangeFloat(Minimum=80, Maximum=550)
    Stone Velocity Cone Angle=120
    StoneScale=0.6
    Wood Gravity=&RandomRangeFloat(Minimum=-450, Maximum=-900)
    Wood Mesh Count=2
    Wood Scale=1.5
    Wood Velocity=&RandomRangeFloat(Minimum=80, Maximum=600)
    Wood Velocity Cone Angle=130
   )
   &AddRotationalVelocity(Rotation Rate=&RandomVector(Vector Scale=&RandomRangeFloat(Minimum=1, Maximum=2)))
   &ApplyInitialForces()
   &AddVelocity(
    Cone Angle=20
    Cone Axis=&Lerp_Vector(
     A=<1 1.5 -0.25>
     B=<-1 1.5 -0.5>
     Alpha=&UniformAOrBFloat(A=0, Evaluation Type=::NewEnumerator0, Recalculate Random Each Loop=true, Override Seed=false)
    )
    Inner Cone Angle=5
    Module.UseConeFalloff=false
    Module.UseVelDistribution=false
    Override Seed=false
    Velocity Mode=::NewEnumerator2
    Velocity Speed=&Lerp_Float(
     A=250
     B=300
     Alpha=&FloatFromCurve(
      FloatCurve=DataInterface() {propertyValues {Curve {keys [
          {interpMode=RCIM_Cubic, tangentMode=RCTM_User, time=0, value=0}
          {interpMode=RCIM_Cubic, tangentMode=RCTM_User, time=0.5, value=1}
          {tangentMode=RCTM_User, time=1, value=0}
         ]}}}
      CurveIndex=&RandomRangeFloat(Minimum=0, Maximum=1)
      Scale Curve=3
     )
    )
   ) {module=AddVelocity003}
   &InitialMeshOrientation(Use Rotation Vector=false)
  ]
  particleUpdateScript [
   &ParticleState()
   &GravityForce()
   &Collision(
    Advanced Aging Rate=1
    Age Colliding Particles=false
    CPU Collision Trace Channel=ECC_WorldStatic
    Collision Enabled=&Boolean_AndOperation(
     Bool1=&SetBoolByFloatComparison(A=$Particles.Initial.Lifetime, B=1)
     Bool2=&SetBoolByFloatComparison(A=$Particles.Age, B=0.12)
    )
    Control Roll On Collision=true
    Friction=1
    Friction During a Bounce=0.25
    Max CPU Trace Length=100
    Module.Control Rest State Via Surface Normal=false
    Module.EnableMaxCollisionCount=false
    Module.Randomize Collision Normal Vector=true
    Randomize Collision Normal=0.25
    ScaleRateOfRotation=false
   )
   &InheritSourceMovement(ApplyForce=false, ApplyPosition=true, ApplyPositionFalloff=false)
   &SolveForcesAndVelocity(Clamp Velocity=false, Limit Acceleration=false)
   &ScaleMeshSize(Scale Factor=&VectorFromFloat(Value=&FloatFromCurve(
      FloatCurve=DataInterface() {propertyValues {Curve {keys [<at 0.6 1 RCIM_Cubic>, {arriveTangent=-6.308, interpMode=RCIM_Cubic, leaveTangent=-6.308, tangentMode=RCTM_User, time=1, value=0}]}}}
      Scale Curve=1.5
     )))
   &DynamicMaterialParameters(Module.Param0WriteEnabled=true, Module.Param1WriteEnabled=true, Module.Param2WriteEnabled=false, Module.Param3WriteEnabled=false)
  ]
  rendererClasses [&NiagaraMeshRendererProperties]
  renderers [{rendererClass=&NiagaraMeshRendererProperties, rendererIndex=0}]
 }
 Ribbon_Ground {
  emitterUpdateScript [
   &EmitterState(Life Cycle Mode=::NewEnumerator1, Loop Behavior=::NewEnumerator1, Loop Delay=20, Loop Duration Mode=::NewEnumerator1, UseLoopDelay=false)
   /Game/Effects/Niagara/Modules/Spawn/FN_SpawnByEngineDistance(
    Multiplier:NiagaraInt32=&MakeCustomIntFromBool(Bool=&SetBoolByFloatComparison(A=$Engine.System.Age, B=0.7), True INT=1, False INT=0)
    Step=10
   )
  ]
  particleSpawnScript [
   &InitializeRibbon(
    Color=&FN_Select_Color_from_Time_Of_Day(
     Morning=<rgba 0.0469 0.035 0.0308 2>
     Day=<rgba 0.0469 0.035 0.0308 2>
     Evening=<rgba 0.0469 0.035 0.0308 2>
     Night=&MultiplyLinearColor(Linear Color=<rgba 0.0469 0.035 0.0308 2>, Scale Factor=<0.25 0.25 0.5 1>)
    )
    Module.Write Color=true
    Module.Write Lifetime=true
    Module.Write Mass=true
    Module.Write Material Random=true
    Module.Write Position=false
    Module.Write RibbonTwist=false
    Module.Write RibbonWidth=true
    Ribbon Width=40
   )
   /SetVariables(Particles.Stream:NiagaraInt32=&SpawnGroup()) {isSetParameters=true}
   &FN_SetSpawnGroup(Number of Spawn Groups:NiagaraInt32=$Emitter.Spawn Multiplier)
   /SetVariables() {isSetParameters=true}
  ]
  particleUpdateScript [
   &ParticleState()
   &InheritSourceMovement(
    Applied Force Scale=&VectorFromFloat(Value=-0.5)
    ApplyForce=true
    ApplyForceFalloff=false
    ApplyPosition=true
    ApplyPositionFalloff=false
    Module.LimitForceSpeed=false
   )
   &SolveForcesAndVelocity(Clamp Velocity=false, Limit Acceleration=false)
   &self:Ribbon_Ground_0.NiagaraScratchPadContainer_1.FN_Lock_Ribbon_Head_0(
    Emitter Count:NiagaraInt32=$Emitter.Count
    Locked Position:Vector3f=&AddVectorToPosition(Position=/Niagara/DynamicInputs/Helpers/SimulationPosition(), Vector=<0 0 2.5>)
    Particle Stream:NiagaraInt32=$Particles.Stream
   ) {module=FN_Lock_Ribbon_Head}
   &ScaleRibbonWidth(Ribbon Width Scale=&FloatFromCurve(FloatCurve=DataInterface() {propertyValues {Curve {keys [{tangentMode=RCTM_User, time=0, value=0}, {tangentMode=RCTM_User, time=1, value=6}]}}}))
   &ScaleColor(Scale Alpha=0.5, ScaleA=true, ScaleRGB=true)
  ]
  rendererClasses [&NiagaraRibbonRendererProperties]
  renderers [{rendererClass=&NiagaraRibbonRendererProperties}]
 }
 Smoke {
  emitterUpdateScript [
   &EmitterState(Life Cycle Mode=::NewEnumerator1, Loop Behavior=::NewEnumerator1, Loop Delay=0.6, Loop Duration Mode=::NewEnumerator1, UseLoopDelay=false)
   &SpawnRate(Spawn Probability=0.8, SpawnRate=20, Use Spawn Probability=true)
   &SpawnBurst_Instantaneous(Spawn Count=&UniformRangedInt(Minimum=2, Maximum=3), Use Spawn Probability=false, UseLoopCountLimit=false)
  ]
  particleSpawnScript [
   &InitializeParticle(
    Color=<rgba 0.7 0.383 0.175 0.75>
    Lifetime Max=0.4
    Lifetime Min=0.2
    Lifetime Mode=::NewEnumerator1
    Mass Mode=::NewEnumerator1
    Sprite Size Mode=::NewEnumerator1
    Uniform Sprite Size Max=150
    Uniform Sprite Size Min=70
    UsePositionOffset=false
    Write Material Random=true
   ) {module=InitializeParticle001}
   &CameraOffset(Camera Offset Amount=&UniformRangedFloat(Minimum=0, Maximum=20))
   &AddVelocity(
    Cone Angle=15
    Cone Axis=<0 1 -0.25>
    Module.UseConeFalloff=false
    Module.UseVelDistribution=false
    Override Seed=false
    Velocity Mode=::NewEnumerator2
    Velocity Speed=&Multiply_Float(A=&RandomRangeFloat(Minimum=25, Maximum=225), B=1)
   ) {module=AddVelocity003}
  ]
  particleUpdateScript [
   &ParticleState()
   /SetVariables(Particles.SubImageIndex=&Lerp_Float(A=0, B=10, Alpha=$Particles.NormalizedAge)) {isSetParameters=true}
   &Drag(Drag=&RandomRangeFloat(Minimum=2, Maximum=3), Use Linear Drag=true, Use Rotational Drag=true)
   &AccelerationForce(Acceleration=&RandomRangeVector(Minimum=<0 0 5>, Maximum=<0 0 35>))
   &ScaleSpriteSize(Scale Factor=&Vector2DFromFloat(Value=&FloatFromCurve(FloatCurve=DataInterface() {propertyValues {Curve {keys [
          {arriveTangent=0.221, interpMode=RCIM_Cubic, leaveTangent=0.221, tangentMode=RCTM_User, time=0, value=0.15}
          {arriveTangent=0.7659, interpMode=RCIM_Cubic, leaveTangent=0.7659, tangentMode=RCTM_User, time=1, value=0.8}
         ]}}})))
   &ScaleColor(
    Scale Alpha=&FloatFromCurve(
     FloatCurve=DataInterface() {propertyValues {Curve {keys [<at 0 0>,<at 0.25 1>,<at 1 0>]}}}
     Scale Curve=&RandomRangeFloat(Minimum=0.15, Maximum=0.1)
    )
    ScaleA=true
    ScaleRGB=true
   )
   &DynamicMaterialParameters(
    Index 0 Param 1=&FloatFromCurve(FloatCurve=DataInterface() {propertyValues {Curve {keys [<at 0.5 0.5>,<at 1 -0.1>]}}})
    Index 0 Param 4=1
    Module.Param0WriteEnabled=true
    Module.Param1WriteEnabled=true
    Module.Param2WriteEnabled=true
    Module.Param3WriteEnabled=true
   )
   &SolveForcesAndVelocity(Clamp Velocity=false, Limit Acceleration=false)
  ]
  rendererClasses [&NiagaraSpriteRendererProperties]
  renderers [{rendererClass=&NiagaraSpriteRendererProperties}]
 }
 Smoke_Spike {
  emitterUpdateScript [
   &EmitterState(Life Cycle Mode=::NewEnumerator1, Loop Behavior=::NewEnumerator1, Loop Delay=0.1, Loop Duration Mode=::NewEnumerator1, UseLoopDelay=true)
   &SpawnRate(Spawn Probability=0.67, SpawnRate=20, Use Spawn Probability=false)
  ]
  particleSpawnScript [
   /SetVariables(Particles.Rand:NiagaraFloat=&UniformAOrBFloat(A=0, Evaluation Type=::NewEnumerator0, Recalculate Random Each Loop=true, Override Seed=false)) {isSetParameters=true, module=SetVariables_#1}
   &InitializeParticle(
    Color=&FN_Select_Color_from_Time_Of_Day(
     Morning=<rgba 0.0469 0.035 0.0308 2>
     Day=<rgba 0.0469 0.035 0.0308 2>
     Evening=<rgba 0.0469 0.035 0.0308 2>
     Night=&MultiplyLinearColor(Linear Color=<rgba 0.0469 0.035 0.0308 2>, Scale Factor=<0.5 0.5 1 1>)
    )
    Lifetime=&RandomRangeFloat(Minimum=0.5, Maximum=0.5)
    Mesh Scale Mode=::NewEnumerator1
    Mesh Uniform Scale Max=3
    Mesh Uniform Scale Min=1.5
    Position Offset=&Lerp_Vector(A=<-5 0 0>, B=<5 0 0>, Alpha=$Particles.Rand)
    UsePositionOffset=true
    Write Material Random=true
   )
   &AddVelocity(
    Cone Angle=30
    Cone Axis=&Lerp_Vector(A=<-1 1 -0.25>, B=<1 1 -0.25>, Alpha=$Particles.Rand)
    Module.UseConeFalloff=false
    Module.UseVelDistribution=false
    Override Seed=false
    Velocity Mode=::NewEnumerator2
    Velocity Speed=&Lerp_Float(
     A=0.15
     B=0.25
     Alpha=&FloatFromCurve(
      FloatCurve=DataInterface() {propertyValues {Curve {keys [
          {interpMode=RCIM_Cubic, tangentMode=RCTM_User, time=0, value=0}
          {interpMode=RCIM_Cubic, tangentMode=RCTM_User, time=0.5, value=1}
          {tangentMode=RCTM_User, time=1, value=0}
         ]}}}
      CurveIndex=&RandomRangeFloat(Minimum=0, Maximum=1)
      Scale Curve=3
     )
    )
   ) {module=AddVelocity003}
  ]
  particleUpdateScript [
   &ParticleState()
   &DynamicMaterialParameters(
    Index 0 Param 1=$Particles.NormalizedAge
    Index 0 Param 2=&RandomRangeFloat(Minimum=1, Maximum=1.5)
    Index 0 Param 3=&RandomRangeFloat(Minimum=0, Maximum=1)
    Index 0 Param 4=&FloatFromCurve(
     FloatCurve=DataInterface() {propertyValues {Curve {keys [
         {arriveTangent=2.9585, interpMode=RCIM_Cubic, leaveTangent=2.9585, tangentMode=RCTM_User, time=0, value=0}
         {arriveTangent=-0.0405, interpMode=RCIM_Cubic, leaveTangent=-0.0405, tangentMode=RCTM_User, time=1, value=1}
        ]}}}
     Scale Curve=&RandomRangeFloat(Minimum=2, Maximum=3)
    )
    Module.Index0Float0Write=true
    Module.Index0Float1Write=true
    Module.Index0Float2Write=true
    Module.Index0Float3Write=true
   )
   &CameraOffset(Camera Offset Amount=20)
   &Drag(Drag=4, Use Linear Drag=true, Use Rotational Drag=true)
   &ScaleMeshSize(Scale Factor=&VectorFromFloat(Value=&FloatFromCurve(
      FloatCurve=DataInterface() {propertyValues {Curve {keys [
          {arriveTangent=3.844, interpMode=RCIM_Cubic, leaveTangent=3.844, tangentMode=RCTM_User, time=0, value=0}
          {arriveTangent=0.1649, interpMode=RCIM_Cubic, leaveTangent=0.1649, tangentMode=RCTM_User, time=0.75, value=1}
         ]}}}
      Scale Curve=0.75
     )))
   &ScaleColor(
    Scale Alpha=&FloatFromCurve(FloatCurve=DataInterface() {propertyValues {Curve {keys [{interpMode=RCIM_Cubic, tangentMode=RCTM_User, time=0.1, value=1}, {interpMode=RCIM_Cubic, tangentMode=RCTM_User, time=1, value=0}]}}})
    Scale RGB=&VectorFromFloat(Value=&FloatFromCurve(FloatCurve=DataInterface() {propertyValues {Curve {keys [{arriveTangent=-3.4125, interpMode=RCIM_Cubic, leaveTangent=-3.4125, tangentMode=RCTM_User, time=0, value=3}, <at 0.35 2>]}}}))
    ScaleA=true
    ScaleRGB=true
   )
   &InheritSourceMovement(ApplyForce=false, ApplyPosition=true, ApplyPositionFalloff=false)
   &SolveForcesAndVelocity(Clamp Velocity=false, Limit Acceleration=false)
  ]
  rendererClasses [&NiagaraMeshRendererProperties]
  renderers [{rendererClass=&NiagaraMeshRendererProperties, rendererIndex=0}]
 }
 Sparks {
  simTarget=GPUComputeSim
  emitterUpdateScript [
   &EmitterState(Life Cycle Mode=::NewEnumerator1, Loop Behavior=::NewEnumerator1, Loop Delay=0.6, Loop Duration Mode=::NewEnumerator1, UseLoopDelay=false)
   &SpawnRate(SpawnRate=200, Use Spawn Probability=false)
  ]
  particleSpawnScript [
   /SetVariables(Particles.SparkIntensity:NiagaraFloat=&FloatFromCurve(
     FloatCurve=DataInterface() {propertyValues {Curve {keys [
         <at 0 1 RCIM_Cubic>
         {arriveTangent=-2.5317, interpMode=RCIM_Cubic, leaveTangent=-2.5317, tangentMode=RCTM_User, time=0.6, value=0.5}
         <at 1 0 RCIM_Cubic>
        ]}}}
     CurveIndex=&RandomRangeFloat(Minimum=0, Maximum=1)
    )) {isSetParameters=true}
   &InitializeParticle(
    Color=&LerpLinearColors(StartColor=<rgba 10 0.2 0 1>, EndColor=<rgba 25 5 0 1>, LerpFactor=$Particles.SparkIntensity)
    Lifetime Max=0.25
    Lifetime Min=0.11
    Lifetime Mode=::NewEnumerator1
    Mass=&Lerp_Float(A=0.3, B=1, Alpha=$Particles.SparkIntensity)
    Mass Mode=::NewEnumerator1
    Sprite Size Mode=::NewEnumerator3
    Uniform Sprite Size=&Lerp_Float(A=2.5, B=3.5, Alpha=$Particles.SparkIntensity)
    UsePositionOffset=false
    Write Material Random=true
   )
   /SetVariables(Particles.SpriteSize=&Multiply_Vector2(
     A=$Particles.SpriteSize
     B=&Vector2DFromFloat(Value=&FloatFromCurve(FloatCurve=DataInterface() {propertyValues {Curve {keys [<at 1000 1 RCIM_Cubic>,<at 5000 2>]}}}, CurveIndex=$Engine.Owner.LODDistance))
    )) {isSetParameters=true}
   &AddVelocity(
    Cone Angle=20
    Cone Axis=<0 1 -0.6>
    Inner Cone Angle=5
    Module.UseConeFalloff=false
    Module.UseVelDistribution=false
    Override Seed=false
    Velocity Mode=::NewEnumerator2
    Velocity Speed=&Multiply_Float(
     A=&Lerp_Float(
      A=250
      B=550
      Alpha=&FloatFromCurve(
       FloatCurve=DataInterface() {propertyValues {Curve {keys [
           {interpMode=RCIM_Cubic, tangentMode=RCTM_User, time=0, value=0}
           {interpMode=RCIM_Cubic, tangentMode=RCTM_User, time=0.5, value=1}
           {tangentMode=RCTM_User, time=1, value=0}
          ]}}}
       CurveIndex=&RandomRangeFloat(Minimum=0, Maximum=1)
      )
     )
     B=3
    )
   ) {module=AddVelocity003}
  ]
  particleUpdateScript [
   &ParticleState()
   &GravityForce(Gravity=<0 0 -1600>)
   &Drag(Drag=0.15, Use Linear Drag=true, Use Rotational Drag=false)
   &VectorNoiseForce(
    Force Amount=&FloatFromCurve(
     FloatCurve=DataInterface() {propertyValues {Curve {keys [
         {arriveTangent=-9.4515, interpMode=RCIM_Cubic, leaveTangent=-4.1687, tangentMode=RCTM_User, time=0, value=1}
         {arriveTangent=-3.4372, interpMode=RCIM_Cubic, leaveTangent=-3.4372, tangentMode=RCTM_User, time=0.6, value=0}
        ]}}}
     Scale Curve=300
    )
    Module.Use Random Noise Time Interval=false
   )
   &InheritSourceMovement(ApplyForce=false, ApplyPosition=true, ApplyPositionFalloff=false)
   &ScaleColor(
    Linear Color Curve=DataInterface() {propertyValues {
      AlphaCurve {keys [<at 0 1>,<at 1 1>]}
      BlueCurve {keys [
        <at 0 2 RCIM_Cubic>
        {arriveTangent=-1.452, interpMode=RCIM_Cubic, leaveTangent=-1.452, tangentMode=RCTM_User, time=0.1467, value=1}
        <at 1 0.0755>
       ]}
      GreenCurve {keys [
        <at 0 2 RCIM_Cubic>
        {arriveTangent=-1.4408, interpMode=RCIM_Cubic, leaveTangent=-1.4408, tangentMode=RCTM_User, time=0.1467, value=1}
        <at 1 0.078>
       ]}
      RedCurve {keys [
        <at 0 2 RCIM_Cubic>
        {arriveTangent=-0.8332, interpMode=RCIM_Cubic, leaveTangent=-0.8332, tangentMode=RCTM_User, time=0.1467, value=1}
        <at 1 0.25>
       ]}
     }}
    Scale Mode=::NewEnumerator2
   )
   &ScaleSpriteSize(Scale Factor=&Vector2DFromFloat(Value=&FloatFromCurve(FloatCurve=DataInterface() {propertyValues {Curve {keys [<at 0.4 1 RCIM_Cubic>, {arriveTangent=-3.8662, interpMode=RCIM_Cubic, leaveTangent=-3.8662, tangentMode=RCTM_User, time=1, value=0}]}}})))
   &SolveForcesAndVelocity(Clamp Velocity=false, Limit Acceleration=false)
   &ScaleSpriteSizeBySpeed(
    Max Scale Factor=<0.6 10>
    Min Scale Factor=<0.5 1>
    Sample Scale Factor By Curve=true
    Scale Factor Curve=DataInterface() {propertyValues {Curve {keys [
        {arriveTangent=0.2757, time=0, value=0}
        {arriveTangent=0.3844, interpMode=RCIM_Cubic, leaveTangent=0.3844, tangentMode=RCTM_User, time=0.3, value=0.15}
        {arriveTangent=2.2765, interpMode=RCIM_Cubic, leaveTangent=2.2765, tangentMode=RCTM_User, time=1, value=1}
       ]}}}
   )
  ]
  rendererClasses [&NiagaraSpriteRendererProperties]
  renderers [{rendererClass=&NiagaraSpriteRendererProperties}]
 }
 SpeedLine {
  emitterUpdateScript [
   &EmitterState(Life Cycle Mode=::NewEnumerator1, Loop Behavior=::NewEnumerator1, Loop Delay=0.2, Loop Duration Mode=::NewEnumerator1, UseLoopDelay=true)
   &SpawnRate(SpawnRate=90, Use Spawn Probability=false) {module=SpawnRate001}
  ]
  particleSpawnScript [
   &InitializeParticle(
    Lifetime Max=0.16
    Lifetime Min=0.1
    Lifetime Mode=::NewEnumerator1
    Sprite Size Max=<3 10>
    Sprite Size Min=<2 2>
    Sprite Size Mode=::NewEnumerator2
    UsePositionOffset=false
    Write Material Random=true
   )
   &ShapeLocation(
    Handle Radius:NiagaraFloat=&RandomRangeFloat(Minimum=5, Maximum=10)
    Large Radius=&RandomRangeFloat(Minimum=150, Maximum=250)
    MaskSpawnGroup=false
    Override Seed=false
    Rotation Angle=90
    Rotation Angle Type:ENiagara_AngleInput=::NewEnumerator0
    Rotation Axis=<0 1 0>
    Rotation Mode=::NewEnumerator1
    Shape Primitive=::NewEnumerator3
    UseOwnerScale=true
   )
   &AddVelocityFromPoint(Module.Use Falloff Distance=false, Origin Offset=<0 -500 250>, Velocity Strength=&Multiply_Float(A=600, B=1))
  ]
  particleUpdateScript [
   &ParticleState()
   &InheritSourceMovement(
    Applied Position Scale=&Multiply_VectorByFloat(
     Vector=<1 1 1>
     Float=&FloatFromCurve(FloatCurve=DataInterface() {propertyValues {Curve {keys [{interpMode=RCIM_Cubic, tangentMode=RCTM_User, time=0, value=1}, {interpMode=RCIM_Cubic, tangentMode=RCTM_User, time=1, value=0}]}}})
    )
    ApplyForce=false
    ApplyPosition=true
    ApplyPositionFalloff=false
   )
   &SolveForcesAndVelocity(Clamp Velocity=false, Limit Acceleration=false)
   &ScaleSpriteSize()
   &ScaleSpriteSizeBySpeed(Max Scale Factor=<1 50>, Min Scale Factor=<0.01 0.05>, Velocity Threshold=1)
   &ScaleColor(
    Scale Alpha=&FloatFromCurve(
     FloatCurve=DataInterface() {propertyValues {Curve {keys [
         <at 0 0 RCIM_Cubic>
         <at 0.3759 1.0207>
         {arriveTangent=-2.221, interpMode=RCIM_Cubic, leaveTangent=-2.221, tangentMode=RCTM_User, time=1, value=0}
        ]}}}
     Scale Curve=0.2
    )
    ScaleA=true
    ScaleRGB=true
   )
  ]
  rendererClasses [&NiagaraSpriteRendererProperties]
  renderers [{rendererClass=&NiagaraSpriteRendererProperties}]
 }
}
summary {
 systemName=NS_Chainsaw_GroundSparks
 emitters [
  {emitterName=Glow_Flare, rendererClasses [&NiagaraSpriteRendererProperties, &NiagaraLightRendererProperties]}
  {emitterName=Smoke, rendererClasses [&NiagaraSpriteRendererProperties]}
  {emitterName=PhysMatDebris, rendererClasses [&NiagaraMeshRendererProperties]}
  {emitterName=Sparks, simTarget=GPUComputeSim, rendererClasses [&NiagaraSpriteRendererProperties]}
  {emitterName=Dirts, rendererClasses [&NiagaraSpriteRendererProperties]}
  {emitterName=Smoke_Spike, rendererClasses [&NiagaraMeshRendererProperties]}
  {emitterName=Ribbon_Ground, rendererClasses [&NiagaraRibbonRendererProperties]}
  {emitterName=Dirt_Chunks, rendererClasses [&NiagaraSpriteRendererProperties]}
  {emitterName=SpeedLine, rendererClasses [&NiagaraSpriteRendererProperties]}
 ]
 userVariables []
}
