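# Niagara ribbon effect that draws sprint trails for the Chainsaw scooter.
# A single emitter spawns two short-lived anim trails from bone or socket indices on the User.SkeletalMesh parameter.
# Color follows time of day, translucent white normally and dimmer blue-tinted at night.
# Ribbon width and alpha taper over particle life, with alpha also multiplied by a system-age curve so trails fade in after the effect starts.
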NiagaraSystem /ChainsawGameplay/Assets/FX/NS_Chainsaw_Scooter_SprintTrail
dependencies {usedDataInterfaces [&NiagaraDataInterfaceCurve], usedDynamicInputs [], usedModules [&SystemState]}
emitters {FN_AnimTrail {
  emitterSpawnScript [&FN_RibbonIDOffsetUtility()]
  emitterUpdateScript [
   &EmitterState(Life Cycle Mode=::NewEnumerator1, Loop Behavior=::NewEnumerator1, Loop Delay=2, Loop Duration Mode=::NewEnumerator1, UseLoopDelay=false)
   &FN_SpawnAnimTrail(Number of Anim Trails=2)
  ]
  particleSpawnScript [
   &FN_InitializeAnimTrailBoneIndecies(First Bone or Socket Index=1, Second Bone or Socket Index=0)
   &FN_InitializeAnimTrail(
    Color=&FN_Select_Color_from_Time_Of_Day(
     Morning=<rgba 1 1 1 0.55>
     Day=<rgba 1 1 1 0.55>
     Evening=<rgba 1 1 1 0.55>
     Night=&MultiplyLinearColor(Linear Color=<rgba 1 1 1 0.55>, Scale Factor=<0.25 0.5 1 0.5>)
    )
    Lifetime=0.25
   )
  ]
  particleUpdateScript [
   &ScaleRibbonWidth(Ribbon Width Scale=&FloatFromCurve(FloatCurve=DataInterface() {propertyValues {Curve {keys [
         {interpMode=RCIM_Cubic, tangentMode=RCTM_User, time=0, value=1}
         {arriveTangent=-3.1724, interpMode=RCIM_Cubic, leaveTangent=-3.1724, tangentMode=RCTM_User, time=1, value=0}
        ]}}}))
   &ScaleColor(
    Scale Alpha=&Multiply_Float(
     A=&FloatFromCurve(
      FloatCurve=DataInterface() {propertyValues {Curve {keys [<at 0 1 RCIM_Cubic>, {arriveTangent=-1.9571, interpMode=RCIM_Cubic, leaveTangent=-1.9571, tangentMode=RCTM_User, time=1, value=0}]}}}
      Scale Curve=1.2
     )
     B=&FloatFromCurve(
      FloatCurve=DataInterface() {propertyValues {Curve {keys [{interpMode=RCIM_Cubic, tangentMode=RCTM_User, time=0.5, value=0}, {interpMode=RCIM_Cubic, tangentMode=RCTM_User, time=2.5, value=1}]}}}
      CurveIndex=$System.Age
     )
    )
    ScaleA=true
    ScaleRGB=true
   ) {module=ScaleColor001}
   &FN_UpdateAnimTrail(Locked Particle RibbonWidth:NiagaraFloat=$Particles.RibbonWidth, SkeletalMesh=$User.SkeletalMesh) {module=FN_UpdateAnimTrail001}
  ]
  rendererClasses [&NiagaraRibbonRendererProperties]
  renderers [{rendererClass=&NiagaraRibbonRendererProperties}]
 }}
summary {
 systemName=NS_Chainsaw_Scooter_SprintTrail
 emitters [{emitterName=FN_AnimTrail, rendererClasses [&NiagaraRibbonRendererProperties]}]
 userVariables [
  User.RibbonIDOffset {type=NiagaraInt32}
  User.SkeletalMesh {
   type=NiagaraDataInterfaceSkeletalMesh
   defaultValue {
    struct=NiagaraExt_VariableValue_DataInterface
    value {dataInterface=&self:NiagaraDataInterfaceSkeletalMesh_0}
   }
  }
 ]
}
