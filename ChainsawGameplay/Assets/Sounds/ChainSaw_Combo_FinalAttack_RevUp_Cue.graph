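# Sound cue for the chainsaw combo final attack rev-up, playing one of four rev-up waves at random.
# IsLocal and IsEnemy branches pick local, enemy or non-enemy paths, with unset parameters falling back to the local path.
# All paths share the same waves through separate random and modulator nodes, with mixers boosting enemies and lowering non-enemies.
# The cue uses close-range melee attenuation, overrides concurrency and sends to a chainsaw ducking bus.
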SoundCue /ChainsawGameplay/Assets/Sounds/ChainSaw_Combo_FinalAttack_RevUp_Cue
root=Branch_0
nodes [
 {id=Branch_0, children [Modulator_0,Branch_1,Modulator_0], properties {boolParameterName=IsLocal}}
 {id=Branch_1, children [Mixer_8,Mixer_6,Modulator_0], properties {boolParameterName=IsEnemy}}
 {id=Mixer_6, children [Modulator_5], properties {inputVolume [0.7]}}
 {id=Mixer_8, children [Modulator_4], properties {inputVolume [1.5]}}
 {id=Modulator_0, children [Random_0]}
 {id=Modulator_4, children [Random_4]}
 {id=Modulator_5, children [Random_5]}
 {id=Random_0, children [WavePlayer_1,WavePlayer_2,WavePlayer_0,WavePlayer_3], properties {weights [1,1,1,1]}}
 {id=Random_4, children [WavePlayer_8, WavePlayer_9, WavePlayer_10, WavePlayer_11], properties {weights [1,1,1,1]}}
 {id=Random_5, children [WavePlayer_12, WavePlayer_13, WavePlayer_22, WavePlayer_23], properties {weights [1,1,1,1]}}
 {id=WavePlayer_0, properties {soundWaveAssetPtr=&/ChainSaw_Attack_RevUps_05}}
 {id=WavePlayer_1, properties {soundWaveAssetPtr=&/ChainSaw_Attack_RevUps_01}}
 {id=WavePlayer_10, properties {soundWaveAssetPtr=&ChainSaw_Attack_RevUps_05}}
 {id=WavePlayer_11, properties {soundWaveAssetPtr=&/ChainSaw_Attack_RevUps_06}}
 {id=WavePlayer_12, properties {soundWaveAssetPtr=&ChainSaw_Attack_RevUps_01}}
 {id=WavePlayer_13, properties {soundWaveAssetPtr=&/ChainSaw_Attack_RevUps_02}}
 {id=WavePlayer_2, properties {soundWaveAssetPtr=&ChainSaw_Attack_RevUps_02}}
 {id=WavePlayer_22, properties {soundWaveAssetPtr=&ChainSaw_Attack_RevUps_05}}
 {id=WavePlayer_23, properties {soundWaveAssetPtr=&ChainSaw_Attack_RevUps_06}}
 {id=WavePlayer_3, properties {soundWaveAssetPtr=&ChainSaw_Attack_RevUps_06}}
 {id=WavePlayer_8, properties {soundWaveAssetPtr=&ChainSaw_Attack_RevUps_01}}
 {id=WavePlayer_9, properties {soundWaveAssetPtr=&ChainSaw_Attack_RevUps_02}}
]
properties {
 attenuationSettings=/Game/Sounds/Attenuation/Fort_Weapon_Melee_Attack_Close_Range_Stereo_Att
 bOverrideConcurrency=true
 maxDistance=10058.75
 soundClassObject=&Fort_WeaponRanged_Fire_3P_PSM
 volumeMultiplier=0.7
 busSends [{
   audioBus=&/AB_Chainsaw_Duck
   maxSendDistance=0
   minSendDistance=1000
   minSendLevel=0
   sendLevel=1
   soundSourceBus=null
  }]
 concurrencyOverrides {maxCount=3}
}
