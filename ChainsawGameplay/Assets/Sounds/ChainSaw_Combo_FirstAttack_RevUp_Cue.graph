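# Sound cue for the chainsaw's first combo attack rev-up, randomly picking one of four rev-up waves through a modulator.
# IsLocal and IsEnemy branches select the mix: local or unset plays unscaled, remote enemies are boosted and remote non-enemies are reduced.
# It uses close-range melee attenuation, overrides concurrency to cap simultaneous voices, and sends to a chainsaw ducking audio bus.
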SoundCue /ChainsawGameplay/Assets/Sounds/ChainSaw_Combo_FirstAttack_RevUp_Cue
root=Branch_0
nodes [
 {id=Branch_0, children [Modulator_0,Branch_1,Modulator_0], properties {boolParameterName=IsLocal}}
 {id=Branch_1, children [Mixer_8,Mixer_6,Modulator_0], properties {boolParameterName=IsEnemy}}
 {id=Mixer_6, children [Modulator_3], properties {inputVolume [0.7]}}
 {id=Mixer_8, children [Modulator_2], properties {inputVolume [1.5]}}
 {id=Modulator_0, children [Random_0]}
 {id=Modulator_2, children [Random_2]}
 {id=Modulator_3, children [Random_3]}
 {id=Random_0, children [WavePlayer_3,WavePlayer_4,WavePlayer_2,WavePlayer_5], properties {weights [1,1,1,1]}}
 {id=Random_2, children [WavePlayer_14, WavePlayer_15, WavePlayer_16, WavePlayer_17], properties {weights [1,1,1,1]}}
 {id=Random_3, children [WavePlayer_18, WavePlayer_19, WavePlayer_20, WavePlayer_21], properties {weights [1,1,1,1]}}
 {id=WavePlayer_14, properties {soundWaveAssetPtr=&/ChainSaw_Attack_RevUps_03}}
 {id=WavePlayer_15, properties {soundWaveAssetPtr=&/ChainSaw_Attack_RevUps_04}}
 {id=WavePlayer_16, properties {soundWaveAssetPtr=&/ChainSaw_Attack_RevUps_07}}
 {id=WavePlayer_17, properties {soundWaveAssetPtr=&/ChainSaw_Attack_RevUps_08}}
 {id=WavePlayer_18, properties {soundWaveAssetPtr=&ChainSaw_Attack_RevUps_03}}
 {id=WavePlayer_19, properties {soundWaveAssetPtr=&ChainSaw_Attack_RevUps_04}}
 {id=WavePlayer_2, properties {soundWaveAssetPtr=&ChainSaw_Attack_RevUps_07}}
 {id=WavePlayer_20, properties {soundWaveAssetPtr=&ChainSaw_Attack_RevUps_07}}
 {id=WavePlayer_21, properties {soundWaveAssetPtr=&ChainSaw_Attack_RevUps_08}}
 {id=WavePlayer_3, properties {soundWaveAssetPtr=&ChainSaw_Attack_RevUps_03}}
 {id=WavePlayer_4, properties {soundWaveAssetPtr=&ChainSaw_Attack_RevUps_04}}
 {id=WavePlayer_5, properties {soundWaveAssetPtr=&ChainSaw_Attack_RevUps_08}}
]
properties {
 attenuationSettings=/Game/Sounds/Attenuation/Fort_Weapon_Melee_Attack_Close_Range_Stereo_Att
 bOverrideConcurrency=true
 maxDistance=10058.75
 soundClassObject=&Fort_WeaponRanged_Fire_3P_PSM
 volumeMultiplier=0.6
 busSends [{
   audioBus=&/AB_Chainsaw_Duck
   maxSendDistance=0
   minSendDistance=1000
   minSendLevel=0
   sendLevel=1
   soundSourceBus=null
  }]
 concurrencyOverrides {maxCount=3}
}
