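# Sound cue for the blade swing of the chainsaw's second combo attack, with close-range melee attenuation.
# IsLocal and IsEnemy branches play local or unset cases at base volume, boost remote enemy swings and lower remote non-enemy swings.
# Each path adds slight volume modulation to a random pick from the same five blade swing waves.
# The cue caps concurrency at a few voices and sends to a chainsaw ducking audio bus.
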SoundCue /ChainsawGameplay/Assets/Sounds/ChainSaw_Combo_SecondAttack_Swing_Cue
root=Branch_0
nodes [
 {id=Branch_0, children [Modulator_1,Branch_1,Modulator_1], properties {boolParameterName=IsLocal}}
 {id=Branch_1, children [Mixer_8,Mixer_6,Modulator_1], properties {boolParameterName=IsEnemy}}
 {id=Mixer_6, children [Modulator_3], properties {inputVolume [0.65]}}
 {id=Mixer_8, children [Modulator_2], properties {inputVolume [1.35]}}
 {id=Modulator_1, children [Random_1], properties {volumeMin=0.85}}
 {id=Modulator_2, children [Random_2], properties {volumeMin=0.85}}
 {id=Modulator_3, children [Random_3], properties {volumeMin=0.85}}
 {
  id=Random_1
  children [WavePlayer_0,WavePlayer_6,WavePlayer_7,WavePlayer_8,WavePlayer_9]
  properties {bShouldExcludeFromBranchCulling=true, weights [1,1,1,1,1]}
 }
 {
  id=Random_2
  children [WavePlayer_15, WavePlayer_16, WavePlayer_17, WavePlayer_18, WavePlayer_19]
  properties {bShouldExcludeFromBranchCulling=true, weights [1,1,1,1,1]}
 }
 {
  id=Random_3
  children [WavePlayer_20, WavePlayer_21, WavePlayer_24, WavePlayer_25, WavePlayer_26]
  properties {bShouldExcludeFromBranchCulling=true, weights [1,1,1,1,1]}
 }
 {id=WavePlayer_0, properties {soundWaveAssetPtr=&/ChainSaw_Swing_Blade_01}}
 {id=WavePlayer_15, properties {soundWaveAssetPtr=&ChainSaw_Swing_Blade_01}}
 {id=WavePlayer_16, properties {soundWaveAssetPtr=&/ChainSaw_Swing_Blade_02}}
 {id=WavePlayer_17, properties {soundWaveAssetPtr=&/ChainSaw_Swing_Blade_03}}
 {id=WavePlayer_18, properties {soundWaveAssetPtr=&/ChainSaw_Swing_Blade_04}}
 {id=WavePlayer_19, properties {soundWaveAssetPtr=&/ChainSaw_Swing_Blade_05}}
 {id=WavePlayer_20, properties {soundWaveAssetPtr=&ChainSaw_Swing_Blade_01}}
 {id=WavePlayer_21, properties {soundWaveAssetPtr=&ChainSaw_Swing_Blade_02}}
 {id=WavePlayer_24, properties {soundWaveAssetPtr=&ChainSaw_Swing_Blade_03}}
 {id=WavePlayer_25, properties {soundWaveAssetPtr=&ChainSaw_Swing_Blade_04}}
 {id=WavePlayer_26, properties {soundWaveAssetPtr=&ChainSaw_Swing_Blade_05}}
 {id=WavePlayer_6, properties {soundWaveAssetPtr=&ChainSaw_Swing_Blade_02}}
 {id=WavePlayer_7, properties {soundWaveAssetPtr=&ChainSaw_Swing_Blade_03}}
 {id=WavePlayer_8, properties {soundWaveAssetPtr=&ChainSaw_Swing_Blade_04}}
 {id=WavePlayer_9, properties {soundWaveAssetPtr=&ChainSaw_Swing_Blade_05}}
]
properties {
 attenuationSettings=/Game/Sounds/Attenuation/Fort_Weapon_Melee_Attack_Close_Range_Stereo_Att
 bOverrideConcurrency=true
 maxDistance=10058.75
 soundClassObject=&Fort_WeaponRanged_Fire_3P_PSM
 volumeMultiplier=0.5
 busSends [{
   audioBus=&/AB_Chainsaw_Duck
   maxSendDistance=0
   minSendDistance=1000
   minSendLevel=0
   sendLevel=1
   soundSourceBus=null
  }]
 concurrencyOverrides {maxCount=3}
}
