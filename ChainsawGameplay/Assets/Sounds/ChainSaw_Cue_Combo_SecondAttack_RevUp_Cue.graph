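# Sound cue for the chainsaw's second combo attack rev-up, playing one of eight random Sprint_End waves through a modulator.
# An IsLocal branch plays the plain version for local or unset, otherwise an IsEnemy branch routes through mixers that make enemies louder and non-enemies quieter.
# Close-range melee attenuation and a concurrency override limit playback, and a bus send feeds the chainsaw ducking bus.
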SoundCue /ChainsawGameplay/Assets/Sounds/ChainSaw_Cue_Combo_SecondAttack_RevUp_Cue
root=Branch_0
nodes [
 {id=Branch_0, children [Modulator_0,Branch_1,Modulator_0], properties {boolParameterName=IsLocal}}
 {id=Branch_1, children [Mixer_8,Mixer_6,Modulator_0], properties {boolParameterName=IsEnemy}}
 {id=Mixer_6, children [Modulator_5], properties {inputVolume [0.7]}}
 {id=Mixer_8, children [Modulator_4], properties {inputVolume [1.5]}}
 {id=Modulator_0, children [Random_1]}
 {id=Modulator_4, children [Random_8]}
 {id=Modulator_5, children [Random_9]}
 {
  id=Random_1
  children [WavePlayer_1,WavePlayer_2,WavePlayer_0,WavePlayer_3,WavePlayer_4,WavePlayer_5,WavePlayer_6,WavePlayer_7]
  properties {weights [1,1,1,1,1,1,1,1]}
 }
 {
  id=Random_8
  children [WavePlayer_26, WavePlayer_27, WavePlayer_28, WavePlayer_29, WavePlayer_30, WavePlayer_31, WavePlayer_32, WavePlayer_33]
  properties {weights [1,1,1,1,1,1,1,1]}
 }
 {
  id=Random_9
  children [WavePlayer_34, WavePlayer_35, WavePlayer_36, WavePlayer_37, WavePlayer_38, WavePlayer_39, WavePlayer_40, WavePlayer_41]
  properties {weights [1,1,1,1,1,1,1,1]}
 }
 {id=WavePlayer_0, properties {soundWaveAssetPtr=&/ChainSaw_Sprint_End_03}}
 {id=WavePlayer_1, properties {soundWaveAssetPtr=&/ChainSaw_Sprint_End_01}}
 {id=WavePlayer_2, properties {soundWaveAssetPtr=&/ChainSaw_Sprint_End_02}}
 {id=WavePlayer_26, properties {soundWaveAssetPtr=&ChainSaw_Sprint_End_01}}
 {id=WavePlayer_27, properties {soundWaveAssetPtr=&ChainSaw_Sprint_End_02}}
 {id=WavePlayer_28, properties {soundWaveAssetPtr=&ChainSaw_Sprint_End_03}}
 {id=WavePlayer_29, properties {soundWaveAssetPtr=&/ChainSaw_Sprint_End_04}}
 {id=WavePlayer_3, properties {soundWaveAssetPtr=&ChainSaw_Sprint_End_04}}
 {id=WavePlayer_30, properties {soundWaveAssetPtr=&/ChainSaw_Sprint_End_05}}
 {id=WavePlayer_31, properties {soundWaveAssetPtr=&/ChainSaw_Sprint_End_06}}
 {id=WavePlayer_32, properties {soundWaveAssetPtr=&/ChainSaw_Sprint_End_07}}
 {id=WavePlayer_33, properties {soundWaveAssetPtr=&/ChainSaw_Sprint_End_08}}
 {id=WavePlayer_34, properties {soundWaveAssetPtr=&ChainSaw_Sprint_End_01}}
 {id=WavePlayer_35, properties {soundWaveAssetPtr=&ChainSaw_Sprint_End_02}}
 {id=WavePlayer_36, properties {soundWaveAssetPtr=&ChainSaw_Sprint_End_03}}
 {id=WavePlayer_37, properties {soundWaveAssetPtr=&ChainSaw_Sprint_End_04}}
 {id=WavePlayer_38, properties {soundWaveAssetPtr=&ChainSaw_Sprint_End_05}}
 {id=WavePlayer_39, properties {soundWaveAssetPtr=&ChainSaw_Sprint_End_06}}
 {id=WavePlayer_4, properties {soundWaveAssetPtr=&ChainSaw_Sprint_End_05}}
 {id=WavePlayer_40, properties {soundWaveAssetPtr=&ChainSaw_Sprint_End_07}}
 {id=WavePlayer_41, properties {soundWaveAssetPtr=&ChainSaw_Sprint_End_08}}
 {id=WavePlayer_5, properties {soundWaveAssetPtr=&ChainSaw_Sprint_End_06}}
 {id=WavePlayer_6, properties {soundWaveAssetPtr=&ChainSaw_Sprint_End_07}}
 {id=WavePlayer_7, properties {soundWaveAssetPtr=&ChainSaw_Sprint_End_08}}
]
properties {
 attenuationSettings=/Game/Sounds/Attenuation/Fort_Weapon_Melee_Attack_Close_Range_Stereo_Att
 bOverrideConcurrency=true
 maxDistance=10058.75
 soundClassObject=&Fort_WeaponRanged_Fire_3P_PSM
 volumeMultiplier=0.5
 busSends [{
   audioBus=&/AB_Chainsaw_Duck
   maxSendDistance=0
   minSendDistance=1000
   minSendLevel=0
   sendLevel=1
   soundSourceBus=null
  }]
 concurrencyOverrides {maxCount=3}
}
