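# Sound cue for dropping the chainsaw weapon, layering a main drop sound with a randomized engine-off sound.
# A mixer plays the fixed drop wave louder alongside a quieter branch that picks one of four equally weighted engine-off variations and modulates it with a lowered minimum pitch.
# It uses shared weapon-drop attenuation, a loot-collect sound class and the weapon-drop stop-oldest concurrency set.
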SoundCue /ChainsawGameplay/Assets/Sounds/ChainSaw_Drop_Cue
root=Mixer_0
nodes [
 {id=Mixer_0, children [WavePlayer_9,Modulator_0], properties {inputVolume [1.2,0.6]}}
 {id=Modulator_0, children [Random_2], properties {pitchMin=0.8}}
 {id=Random_2, children [WavePlayer_0,WavePlayer_2,WavePlayer_7,WavePlayer_8], properties {weights [1,1,1,1]}}
 {id=WavePlayer_0, properties {soundWaveAssetPtr=&/ChainSaw_Drop_Engine_Off_01}}
 {id=WavePlayer_2, properties {soundWaveAssetPtr=&/ChainSaw_Drop_Engine_Off_02}}
 {id=WavePlayer_7, properties {soundWaveAssetPtr=&/ChainSaw_Drop_Engine_Off_03}}
 {id=WavePlayer_8, properties {soundWaveAssetPtr=&/ChainSaw_Drop_Engine_Off_04}}
 {id=WavePlayer_9, properties {soundWaveAssetPtr=&/CahinSaw_Drop_01}}
]
properties {
 attenuationSettings=&Fort_Weapon_Drop_Att
 maxDistance=3000
 soundClassObject=&Fort_Loot_Collect
 volumeMultiplier=0.9
 concurrencyOverrides {resolutionRule=PreventNew}
 concurrencySet [&Weapon_Drop_StopOldest]
}
