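# Sound cue that plays chainsaw movement foley for the player, with concurrency limited by its own override.
# An IsLocal branch selects first-person attenuation when true and third-person when false or unset.
# Both paths share a slight upward pitch modulator feeding a quality switch.
# The switch's first two slots randomize across ten clips, while the last two use a three-clip subset excluded from branch culling.
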SoundCue /ChainsawGameplay/Assets/Sounds/ChainSaw_Foley_Cue
root=Branch_0
nodes [
 {id=Attenuation_0, children [Modulator_0], properties {attenuationSettings=/Game/Sounds/Fort_Player/Footsteps/Settings/Fort_Player_FS_1P_Att}}
 {id=Attenuation_1, children [Modulator_0], properties {attenuationSettings=/Game/Sounds/Fort_Player/Footsteps/Settings/Fort_Player_FS_3P_Att}}
 {id=Branch_0, children [Attenuation_0, Attenuation_1, Attenuation_1], properties {boolParameterName=IsLocal}}
 {id=Modulator_0, children [QualityLevel_1], properties {pitchMax=1.1, pitchMin=1}}
 {id=QualityLevel_1, children [Random_2,Random_2,Random_3,Random_3]}
 {
  id=Random_2
  children [WavePlayer_0, WavePlayer_1, WavePlayer_2, WavePlayer_3, WavePlayer_4, WavePlayer_5, WavePlayer_6, WavePlayer_7, WavePlayer_8, WavePlayer_30]
  properties {weights [1,1,1,1,1,1,1,1,1,1]}
 }
 {id=Random_3, children [WavePlayer_5, WavePlayer_7, WavePlayer_30], properties {bShouldExcludeFromBranchCulling=true, weights [1,1,1]}}
 {id=WavePlayer_0, properties {soundWaveAssetPtr=&/ChainSaw_Movement_Foley_01}}
 {id=WavePlayer_1, properties {soundWaveAssetPtr=&/ChainSaw_Movement_Foley_02}}
 {id=WavePlayer_2, properties {soundWaveAssetPtr=&/ChainSaw_Movement_Foley_03}}
 {id=WavePlayer_3, properties {soundWaveAssetPtr=&/ChainSaw_Movement_Foley_04}}
 {id=WavePlayer_30, properties {soundWaveAssetPtr=&/ChainSaw_Movement_Foley_10}}
 {id=WavePlayer_4, properties {soundWaveAssetPtr=&/ChainSaw_Movement_Foley_05}}
 {id=WavePlayer_5, properties {soundWaveAssetPtr=&/ChainSaw_Movement_Foley_06}}
 {id=WavePlayer_6, properties {soundWaveAssetPtr=&/ChainSaw_Movement_Foley_07}}
 {id=WavePlayer_7, properties {soundWaveAssetPtr=&/ChainSaw_Movement_Foley_08}}
 {id=WavePlayer_8, properties {soundWaveAssetPtr=&/ChainSaw_Movement_Foley_09}}
]
properties {
 bOverrideConcurrency=true
 maxDistance=3584
 soundClassObject=&Fort_Player_Foley_1P_Athena
 volumeMultiplier=0.7
 concurrencyOverrides {maxCount=4}
}
