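# Sound cue for the chainsaw ground sound in ChainsawGameplay, built on a single wave, ChainSaw_Ground_01.
# The wave player feeds a Modulator node left at default settings, with the cue volume slightly reduced.
# It borrows close-range stereo melee attenuation and the loot-collect sound class.
# It references the PickupMaximum_Oldest concurrency set and also carries single-instance, prevent-new override values, though no flag enabling that override is shown.
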SoundCue /ChainsawGameplay/Assets/Sounds/ChainSaw_Ground_Cue
root=Modulator_0
nodes [
 {id=Modulator_0, children [WavePlayer_0]}
 {id=WavePlayer_0, properties {soundWaveAssetPtr=&/ChainSaw_Ground_01}}
]
properties {
 attenuationSettings=/Game/Sounds/Attenuation/Fort_Weapon_Melee_Attack_Close_Range_Stereo_Att
 maxDistance=10058.75
 soundClassObject=&Fort_Loot_Collect
 volumeMultiplier=0.8
 concurrencyOverrides {maxCount=1, resolutionRule=PreventNew}
 concurrencySet [&PickupMaximum_Oldest]
}
