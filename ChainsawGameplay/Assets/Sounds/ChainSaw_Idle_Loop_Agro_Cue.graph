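# Sound cue for the aggressive chainsaw idle motor, layering looping idle sound with a delayed rev-up, pitched-up swing and spark one-shots.
# Each stack mixes two randomized loops under an envelope that fades in, then slowly decays, while Delay nodes stagger the randomized one-shots.
# IsLocal and IsEnemy branches select three duplicate stacks, with enemies slightly boosted and non-local non-enemies quieter.
# Motor aggro attenuation and a three-voice concurrency override apply.
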SoundCue /ChainsawGameplay/Assets/Sounds/ChainSaw_Idle_Loop_Agro_Cue
root=Branch_0
nodes [
 {id=Branch_0, children [Mixer_2,Branch_1,Mixer_2], properties {boolParameterName=IsLocal}}
 {id=Branch_1, children [Mixer_8,Mixer_6,Mixer_2], properties {boolParameterName=IsEnemy}}
 {id=Delay_0, children [Modulator_1], properties {delayMax=0.15, delayMin=0.15}}
 {id=Delay_1, children [Random_0], properties {delayMax=0.5, delayMin=0.5}}
 {id=Delay_10, children [Random_14], properties {delayMax=0.5, delayMin=0.5}}
 {id=Delay_11, children [Random_15], properties {delayMax=0.6, delayMin=0.6}}
 {id=Delay_2, children [Modulator_2], properties {delayMax=0.5, delayMin=0.5}}
 {id=Delay_3, children [Random_13], properties {delayMax=0.6, delayMin=0.6}}
 {id=Delay_4, children [Modulator_4], properties {delayMax=0.15, delayMin=0.15}}
 {id=Delay_5, children [Modulator_5], properties {delayMax=0.5, delayMin=0.5}}
 {id=Delay_6, children [Random_7], properties {delayMax=0.5, delayMin=0.5}}
 {id=Delay_7, children [Random_8], properties {delayMax=0.6, delayMin=0.6}}
 {id=Delay_8, children [Modulator_7], properties {delayMax=0.15, delayMin=0.15}}
 {id=Delay_9, children [Modulator_8], properties {delayMax=0.5, delayMin=0.5}}
 {
  id=Enveloper_0
  children [Modulator_0]
  properties {
   pitchCurve {editorCurveData {keys [{time=0}]}}
   volumeCurve {editorCurveData {keys [<at 0 0>,<at 1 1>,<at 4.0978 0.9899>,<at 12.0948 0.7>,<at 30 0.35>]}}
  }
 }
 {
  id=Enveloper_1
  children [Modulator_3]
  properties {
   pitchCurve {editorCurveData {keys [{time=0}]}}
   volumeCurve {editorCurveData {keys [<at 0 0>,<at 1 1>,<at 4.0978 0.9899>,<at 12.0948 0.7>,<at 30 0.35>]}}
  }
 }
 {
  id=Enveloper_2
  children [Delay_0]
  properties {
   pitchCurve {editorCurveData {keys [{time=0}]}}
   volumeCurve {editorCurveData {keys [<at 0 0>,<at 0.484 0.9821>,<at 4.0978 0.9899>,<at 12.0948 0.7>,<at 30 0.35>]}}
  }
 }
 {
  id=Enveloper_3
  children [Delay_4]
  properties {
   pitchCurve {editorCurveData {keys [{time=0}]}}
   volumeCurve {editorCurveData {keys [<at 0 0>,<at 0.484 0.9821>,<at 4.0978 0.9899>,<at 12.0948 0.7>,<at 30 0.35>]}}
  }
 }
 {
  id=Enveloper_4
  children [Modulator_6]
  properties {
   pitchCurve {editorCurveData {keys [{time=0}]}}
   volumeCurve {editorCurveData {keys [<at 0 0>,<at 1 1>,<at 4.0978 0.9899>,<at 12.0948 0.7>,<at 30 0.35>]}}
  }
 }
 {
  id=Enveloper_5
  children [Delay_8]
  properties {
   pitchCurve {editorCurveData {keys [{time=0}]}}
   volumeCurve {editorCurveData {keys [<at 0 0>,<at 0.484 0.9821>,<at 4.0978 0.9899>,<at 12.0948 0.7>,<at 30 0.35>]}}
  }
 }
 {id=Mixer_0, children [Random_9,Random_10], properties {inputVolume [1,0.5]}}
 {id=Mixer_1, children [Random_3,Random_4], properties {inputVolume [1,0.5]}}
 {id=Mixer_2, children [Enveloper_0,Enveloper_2,Delay_2,Delay_1,Delay_3], properties {inputVolume [1,0.85,0.7,1.25,0.6]}}
 {id=Mixer_3, children [Enveloper_1,Enveloper_3,Delay_5,Delay_6,Delay_7], properties {inputVolume [1,0.85,0.7,1.25,0.6]}}
 {id=Mixer_4, children [Random_16,Random_17], properties {inputVolume [1,0.5]}}
 {id=Mixer_5, children [Enveloper_4,Enveloper_5,Delay_9,Delay_10,Delay_11], properties {inputVolume [1,0.85,0.7,1.25,0.6]}}
 {id=Mixer_6, children [Mixer_5], properties {inputVolume [0.7]}}
 {id=Mixer_8, children [Mixer_3], properties {inputVolume [1.05]}}
 {id=Modulator_0, children [Mixer_1], properties {pitchMax=1}}
 {id=Modulator_1, children [Random_1], properties {pitchMin=1}}
 {id=Modulator_2, children [Random_2], properties {pitchMax=1.3, pitchMin=1.25}}
 {id=Modulator_3, children [Mixer_0], properties {pitchMax=1}}
 {id=Modulator_4, children [Random_5], properties {pitchMin=1}}
 {id=Modulator_5, children [Random_6], properties {pitchMax=1.3, pitchMin=1.25}}
 {id=Modulator_6, children [Mixer_4], properties {pitchMax=1}}
 {id=Modulator_7, children [Random_11], properties {pitchMin=1}}
 {id=Modulator_8, children [Random_12], properties {pitchMax=1.3, pitchMin=1.25}}
 {id=Random_0, children [WavePlayer_1,WavePlayer_2,WavePlayer_6], properties {weights [1,1,1]}}
 {id=Random_1, children [WavePlayer_4,WavePlayer_5], properties {weights [1,1]}}
 {id=Random_10, children [WavePlayer_41, WavePlayer_42, WavePlayer_43, WavePlayer_44, WavePlayer_22], properties {weights [1,1,1,1,1]}}
 {id=Random_11, children [WavePlayer_50, WavePlayer_51], properties {weights [1,1]}}
 {
  id=Random_12
  children [WavePlayer_52, WavePlayer_53, WavePlayer_54, WavePlayer_55, WavePlayer_56, WavePlayer_57]
  properties {weights [1,1,1,1,1,1]}
 }
 {id=Random_13, children [WavePlayer_45, WavePlayer_46, WavePlayer_47], properties {weights [1,1,1]}}
 {id=Random_14, children [WavePlayer_58, WavePlayer_59, WavePlayer_60], properties {weights [1,1,1]}}
 {id=Random_15, children [WavePlayer_61, WavePlayer_62, WavePlayer_63], properties {weights [1,1,1]}}
 {id=Random_16, children [WavePlayer_48, WavePlayer_64, WavePlayer_65, WavePlayer_66, WavePlayer_67], properties {weights [1,1,1,1,1]}}
 {id=Random_17, children [WavePlayer_68, WavePlayer_69, WavePlayer_70, WavePlayer_71, WavePlayer_49], properties {weights [1,1,1,1,1]}}
 {
  id=Random_2
  children [WavePlayer_8, WavePlayer_9, WavePlayer_10, WavePlayer_13, WavePlayer_14, WavePlayer_15]
  properties {weights [1,1,1,1,1,1]}
 }
 {id=Random_3, children [WavePlayer_0, WavePlayer_7, WavePlayer_11, WavePlayer_12, WavePlayer_16], properties {weights [1,1,1,1,1]}}
 {id=Random_4, children [WavePlayer_17, WavePlayer_18, WavePlayer_19, WavePlayer_20, WavePlayer_3], properties {weights [1,1,1,1,1]}}
 {id=Random_5, children [WavePlayer_23, WavePlayer_24], properties {weights [1,1]}}
 {
  id=Random_6
  children [WavePlayer_25, WavePlayer_26, WavePlayer_27, WavePlayer_28, WavePlayer_29, WavePlayer_30]
  properties {weights [1,1,1,1,1,1]}
 }
 {id=Random_7, children [WavePlayer_31, WavePlayer_32, WavePlayer_33], properties {weights [1,1,1]}}
 {id=Random_8, children [WavePlayer_34, WavePlayer_35, WavePlayer_36], properties {weights [1,1,1]}}
 {id=Random_9, children [WavePlayer_21, WavePlayer_37, WavePlayer_38, WavePlayer_39, WavePlayer_40], properties {weights [1,1,1,1,1]}}
 {id=WavePlayer_0, properties {bLooping=true, soundWaveAssetPtr=&/ChainSaw_Idle_Loop_02}}
 {id=WavePlayer_1, properties {soundWaveAssetPtr=&/ChainSaw_Spark_Burst_01}}
 {id=WavePlayer_10, properties {soundWaveAssetPtr=&/ChainSaw_Swing_03}}
 {id=WavePlayer_11, properties {bLooping=true, soundWaveAssetPtr=&/ChainSaw_Idle_Loop_B_02}}
 {id=WavePlayer_12, properties {bLooping=true, soundWaveAssetPtr=&/ChainSaw_Idle_Loop_B_03}}
 {id=WavePlayer_13, properties {soundWaveAssetPtr=&/ChainSaw_Swing_04}}
 {id=WavePlayer_14, properties {soundWaveAssetPtr=&/ChainSaw_Swing_05}}
 {id=WavePlayer_15, properties {soundWaveAssetPtr=&/ChainSaw_Swing_06}}
 {id=WavePlayer_16, properties {bLooping=true, soundWaveAssetPtr=&/ChainSaw_Idle_Loop_B_04}}
 {id=WavePlayer_17, properties {bLooping=true, soundWaveAssetPtr=&/ChainSaw_Idle_Loop_C_01}}
 {id=WavePlayer_18, properties {bLooping=true, soundWaveAssetPtr=&/ChainSaw_Idle_Loop_C_02}}
 {id=WavePlayer_19, properties {bLooping=true, soundWaveAssetPtr=&/ChainSaw_Idle_Loop_C_03}}
 {id=WavePlayer_2, properties {soundWaveAssetPtr=&/ChainSaw_Spark_Burst_02}}
 {id=WavePlayer_20, properties {bLooping=true, soundWaveAssetPtr=&/ChainSaw_Idle_Loop_C_04}}
 {id=WavePlayer_21, properties {bLooping=true, soundWaveAssetPtr=&ChainSaw_Idle_Loop_02}}
 {id=WavePlayer_22, properties {bLooping=true, soundWaveAssetPtr=&/ChainSaw_Idle_Loop_03}}
 {id=WavePlayer_23, properties {soundWaveAssetPtr=&/ChainSaw_Sprint_RevUp_01}}
 {id=WavePlayer_24, properties {soundWaveAssetPtr=&/ChainSaw_Sprint_RevUp_02}}
 {id=WavePlayer_25, properties {soundWaveAssetPtr=&/ChainSaw_Swing_01}}
 {id=WavePlayer_26, properties {soundWaveAssetPtr=&/ChainSaw_Swing_02}}
 {id=WavePlayer_27, properties {soundWaveAssetPtr=&ChainSaw_Swing_03}}
 {id=WavePlayer_28, properties {soundWaveAssetPtr=&ChainSaw_Swing_04}}
 {id=WavePlayer_29, properties {soundWaveAssetPtr=&ChainSaw_Swing_05}}
 {id=WavePlayer_3, properties {bLooping=true, soundWaveAssetPtr=&ChainSaw_Idle_Loop_03}}
 {id=WavePlayer_30, properties {soundWaveAssetPtr=&ChainSaw_Swing_06}}
 {id=WavePlayer_31, properties {soundWaveAssetPtr=&ChainSaw_Spark_Burst_01}}
 {id=WavePlayer_32, properties {soundWaveAssetPtr=&ChainSaw_Spark_Burst_02}}
 {id=WavePlayer_33, properties {soundWaveAssetPtr=&/ChainSaw_Spark_Burst_03}}
 {id=WavePlayer_34, properties {soundWaveAssetPtr=&/ChainSaw_Sparks_01}}
 {id=WavePlayer_35, properties {soundWaveAssetPtr=&/ChainSaw_Sparks_02}}
 {id=WavePlayer_36, properties {soundWaveAssetPtr=&/ChainSaw_Sparks_04}}
 {id=WavePlayer_37, properties {bLooping=true, soundWaveAssetPtr=&/ChainSaw_Idle_Loop_B_01}}
 {id=WavePlayer_38, properties {bLooping=true, soundWaveAssetPtr=&ChainSaw_Idle_Loop_B_02}}
 {id=WavePlayer_39, properties {bLooping=true, soundWaveAssetPtr=&ChainSaw_Idle_Loop_B_03}}
 {id=WavePlayer_4, properties {soundWaveAssetPtr=&ChainSaw_Sprint_RevUp_01}}
 {id=WavePlayer_40, properties {bLooping=true, soundWaveAssetPtr=&ChainSaw_Idle_Loop_B_04}}
 {id=WavePlayer_41, properties {bLooping=true, soundWaveAssetPtr=&ChainSaw_Idle_Loop_C_01}}
 {id=WavePlayer_42, properties {bLooping=true, soundWaveAssetPtr=&ChainSaw_Idle_Loop_C_02}}
 {id=WavePlayer_43, properties {bLooping=true, soundWaveAssetPtr=&ChainSaw_Idle_Loop_C_03}}
 {id=WavePlayer_44, properties {bLooping=true, soundWaveAssetPtr=&ChainSaw_Idle_Loop_C_04}}
 {id=WavePlayer_45, properties {soundWaveAssetPtr=&ChainSaw_Sparks_01}}
 {id=WavePlayer_46, properties {soundWaveAssetPtr=&ChainSaw_Sparks_02}}
 {id=WavePlayer_47, properties {soundWaveAssetPtr=&ChainSaw_Sparks_04}}
 {id=WavePlayer_48, properties {bLooping=true, soundWaveAssetPtr=&ChainSaw_Idle_Loop_02}}
 {id=WavePlayer_49, properties {bLooping=true, soundWaveAssetPtr=&ChainSaw_Idle_Loop_03}}
 {id=WavePlayer_5, properties {soundWaveAssetPtr=&ChainSaw_Sprint_RevUp_02}}
 {id=WavePlayer_50, properties {soundWaveAssetPtr=&ChainSaw_Sprint_RevUp_01}}
 {id=WavePlayer_51, properties {soundWaveAssetPtr=&ChainSaw_Sprint_RevUp_02}}
 {id=WavePlayer_52, properties {soundWaveAssetPtr=&ChainSaw_Swing_01}}
 {id=WavePlayer_53, properties {soundWaveAssetPtr=&ChainSaw_Swing_02}}
 {id=WavePlayer_54, properties {soundWaveAssetPtr=&ChainSaw_Swing_03}}
 {id=WavePlayer_55, properties {soundWaveAssetPtr=&ChainSaw_Swing_04}}
 {id=WavePlayer_56, properties {soundWaveAssetPtr=&ChainSaw_Swing_05}}
 {id=WavePlayer_57, properties {soundWaveAssetPtr=&ChainSaw_Swing_06}}
 {id=WavePlayer_58, properties {soundWaveAssetPtr=&ChainSaw_Spark_Burst_01}}
 {id=WavePlayer_59, properties {soundWaveAssetPtr=&ChainSaw_Spark_Burst_02}}
 {id=WavePlayer_6, properties {soundWaveAssetPtr=&ChainSaw_Spark_Burst_03}}
 {id=WavePlayer_60, properties {soundWaveAssetPtr=&ChainSaw_Spark_Burst_03}}
 {id=WavePlayer_61, properties {soundWaveAssetPtr=&ChainSaw_Sparks_01}}
 {id=WavePlayer_62, properties {soundWaveAssetPtr=&ChainSaw_Sparks_02}}
 {id=WavePlayer_63, properties {soundWaveAssetPtr=&ChainSaw_Sparks_04}}
 {id=WavePlayer_64, properties {bLooping=true, soundWaveAssetPtr=&ChainSaw_Idle_Loop_B_01}}
 {id=WavePlayer_65, properties {bLooping=true, soundWaveAssetPtr=&ChainSaw_Idle_Loop_B_02}}
 {id=WavePlayer_66, properties {bLooping=true, soundWaveAssetPtr=&ChainSaw_Idle_Loop_B_03}}
 {id=WavePlayer_67, properties {bLooping=true, soundWaveAssetPtr=&ChainSaw_Idle_Loop_B_04}}
 {id=WavePlayer_68, properties {bLooping=true, soundWaveAssetPtr=&ChainSaw_Idle_Loop_C_01}}
 {id=WavePlayer_69, properties {bLooping=true, soundWaveAssetPtr=&ChainSaw_Idle_Loop_C_02}}
 {id=WavePlayer_7, properties {bLooping=true, soundWaveAssetPtr=&ChainSaw_Idle_Loop_B_01}}
 {id=WavePlayer_70, properties {bLooping=true, soundWaveAssetPtr=&ChainSaw_Idle_Loop_C_03}}
 {id=WavePlayer_71, properties {bLooping=true, soundWaveAssetPtr=&ChainSaw_Idle_Loop_C_04}}
 {id=WavePlayer_8, properties {soundWaveAssetPtr=&ChainSaw_Swing_01}}
 {id=WavePlayer_9, properties {soundWaveAssetPtr=&ChainSaw_Swing_02}}
]
properties {
 attenuationSettings=&/ChainSaw_Motor_Aggro_Att
 bOverrideConcurrency=true
 maxDistance=5000
 soundClassObject=&Fort_WeaponRanged_Fire_3P_PSM
 volumeMultiplier=0.5
 concurrencyOverrides {maxCount=3}
}
