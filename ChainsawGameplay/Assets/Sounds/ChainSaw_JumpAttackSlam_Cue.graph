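# Sound cue for the chainsaw's jump-attack slam.
# Each variant mixes randomized Impact A and B layers with briefly delayed debris and enemy-hit layers.
# IsLocal and IsEnemy branches play the base mix for local or unset cases, boosting it for remote enemies and lowering it for other remote players.
# It uses AOE-slam attenuation, caps concurrency and sends to a chainsaw ducking bus.
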SoundCue /ChainsawGameplay/Assets/Sounds/ChainSaw_JumpAttackSlam_Cue
root=Branch_3
nodes [
 {id=Branch_1, children [Mixer_8,Mixer_6,Mixer_0], properties {boolParameterName=IsEnemy}}
 {id=Branch_3, children [Mixer_0,Branch_1,Mixer_0], properties {boolParameterName=IsLocal}}
 {id=Delay_0, children [Random_4], properties {delayMax=0.3, delayMin=0.2}}
 {id=Delay_1, children [Random_0], properties {delayMax=0.2, delayMin=0.1}}
 {id=Delay_2, children [Random_15], properties {delayMax=0.3, delayMin=0.2}}
 {id=Delay_3, children [Random_16], properties {delayMax=0.2, delayMin=0.1}}
 {id=Delay_4, children [Random_19], properties {delayMax=0.3, delayMin=0.2}}
 {id=Delay_5, children [Random_20], properties {delayMax=0.2, delayMin=0.1}}
 {id=Mixer_0, children [Random_1,Random_2,Delay_0,Delay_1], properties {inputVolume [0.9,1,1,0.5]}}
 {id=Mixer_4, children [Random_13,Random_14,Delay_2,Delay_3], properties {inputVolume [0.9,1,1,0.5]}}
 {id=Mixer_5, children [Random_17,Random_18,Delay_4,Delay_5], properties {inputVolume [0.9,1,1,0.5]}}
 {id=Mixer_6, children [Mixer_5], properties {inputVolume [0.7]}}
 {id=Mixer_8, children [Mixer_4], properties {inputVolume [1.25]}}
 {id=Random_0, children [WavePlayer_0,WavePlayer_1,WavePlayer_3,WavePlayer_4], properties {weights [1,1,1,1]}}
 {id=Random_1, children [WavePlayer_2,WavePlayer_5,WavePlayer_6,WavePlayer_7], properties {weights [1,1,1,1]}}
 {id=Random_13, children [WavePlayer_42, WavePlayer_43, WavePlayer_44, WavePlayer_45], properties {weights [1,1,1,1]}}
 {id=Random_14, children [WavePlayer_46, WavePlayer_47, WavePlayer_48], properties {weights [1,1,1]}}
 {id=Random_15, children [WavePlayer_49, WavePlayer_50, WavePlayer_51, WavePlayer_52], properties {weights [1,1,1,1]}}
 {id=Random_16, children [WavePlayer_53, WavePlayer_54, WavePlayer_55, WavePlayer_56], properties {weights [1,1,1,1]}}
 {id=Random_17, children [WavePlayer_57, WavePlayer_58, WavePlayer_59, WavePlayer_60], properties {weights [1,1,1,1]}}
 {id=Random_18, children [WavePlayer_61, WavePlayer_62, WavePlayer_63], properties {weights [1,1,1]}}
 {id=Random_19, children [WavePlayer_64, WavePlayer_65, WavePlayer_66, WavePlayer_67], properties {weights [1,1,1,1]}}
 {id=Random_2, children [WavePlayer_8, WavePlayer_12, WavePlayer_16], properties {weights [1,1,1]}}
 {id=Random_20, children [WavePlayer_68, WavePlayer_69, WavePlayer_70, WavePlayer_71], properties {weights [1,1,1,1]}}
 {id=Random_4, children [WavePlayer_17, WavePlayer_21, WavePlayer_22, WavePlayer_23], properties {weights [1,1,1,1]}}
 {id=WavePlayer_0, properties {soundWaveAssetPtr=&/ChainSaw_Scooter_Hit_Enemy_01}}
 {id=WavePlayer_1, properties {soundWaveAssetPtr=&/ChainSaw_Scooter_Hit_Enemy_02}}
 {id=WavePlayer_12, properties {soundWaveAssetPtr=&/ChainSaw_Impact_B_02}}
 {id=WavePlayer_16, properties {soundWaveAssetPtr=&/ChainSaw_Impact_B_03}}
 {id=WavePlayer_17, properties {soundWaveAssetPtr=&/ChainSaw_Hit_Debris_01}}
 {id=WavePlayer_2, properties {soundWaveAssetPtr=&/ChainSaw_Impact_A_01}}
 {id=WavePlayer_21, properties {soundWaveAssetPtr=&/ChainSaw_Hit_Debris_02}}
 {id=WavePlayer_22, properties {soundWaveAssetPtr=&/ChainSaw_Hit_Debris_03}}
 {id=WavePlayer_23, properties {soundWaveAssetPtr=&/ChainSaw_Hit_Debris_04}}
 {id=WavePlayer_3, properties {soundWaveAssetPtr=&/ChainSaw_Scooter_Hit_Enemy_03}}
 {id=WavePlayer_4, properties {soundWaveAssetPtr=&/ChainSaw_Scooter_Hit_Enemy_04}}
 {id=WavePlayer_42, properties {soundWaveAssetPtr=&ChainSaw_Impact_A_01}}
 {id=WavePlayer_43, properties {soundWaveAssetPtr=&/ChainSaw_Impact_A_02}}
 {id=WavePlayer_44, properties {soundWaveAssetPtr=&ChainSaw_Impact_A_01}}
 {id=WavePlayer_45, properties {soundWaveAssetPtr=&/ChainSaw_Impact_A_04}}
 {id=WavePlayer_46, properties {soundWaveAssetPtr=&/ChainSaw_Impact_B_01}}
 {id=WavePlayer_47, properties {soundWaveAssetPtr=&ChainSaw_Impact_B_02}}
 {id=WavePlayer_48, properties {soundWaveAssetPtr=&ChainSaw_Impact_B_03}}
 {id=WavePlayer_49, properties {soundWaveAssetPtr=&ChainSaw_Hit_Debris_01}}
 {id=WavePlayer_5, properties {soundWaveAssetPtr=&ChainSaw_Impact_A_02}}
 {id=WavePlayer_50, properties {soundWaveAssetPtr=&ChainSaw_Hit_Debris_02}}
 {id=WavePlayer_51, properties {soundWaveAssetPtr=&ChainSaw_Hit_Debris_03}}
 {id=WavePlayer_52, properties {soundWaveAssetPtr=&ChainSaw_Hit_Debris_04}}
 {id=WavePlayer_53, properties {soundWaveAssetPtr=&ChainSaw_Scooter_Hit_Enemy_01}}
 {id=WavePlayer_54, properties {soundWaveAssetPtr=&ChainSaw_Scooter_Hit_Enemy_02}}
 {id=WavePlayer_55, properties {soundWaveAssetPtr=&ChainSaw_Scooter_Hit_Enemy_03}}
 {id=WavePlayer_56, properties {soundWaveAssetPtr=&ChainSaw_Scooter_Hit_Enemy_04}}
 {id=WavePlayer_57, properties {soundWaveAssetPtr=&ChainSaw_Impact_A_01}}
 {id=WavePlayer_58, properties {soundWaveAssetPtr=&ChainSaw_Impact_A_02}}
 {id=WavePlayer_59, properties {soundWaveAssetPtr=&ChainSaw_Impact_A_01}}
 {id=WavePlayer_6, properties {soundWaveAssetPtr=&ChainSaw_Impact_A_01}}
 {id=WavePlayer_60, properties {soundWaveAssetPtr=&ChainSaw_Impact_A_04}}
 {id=WavePlayer_61, properties {soundWaveAssetPtr=&ChainSaw_Impact_B_01}}
 {id=WavePlayer_62, properties {soundWaveAssetPtr=&ChainSaw_Impact_B_02}}
 {id=WavePlayer_63, properties {soundWaveAssetPtr=&ChainSaw_Impact_B_03}}
 {id=WavePlayer_64, properties {soundWaveAssetPtr=&ChainSaw_Hit_Debris_01}}
 {id=WavePlayer_65, properties {soundWaveAssetPtr=&ChainSaw_Hit_Debris_02}}
 {id=WavePlayer_66, properties {soundWaveAssetPtr=&ChainSaw_Hit_Debris_03}}
 {id=WavePlayer_67, properties {soundWaveAssetPtr=&ChainSaw_Hit_Debris_04}}
 {id=WavePlayer_68, properties {soundWaveAssetPtr=&ChainSaw_Scooter_Hit_Enemy_01}}
 {id=WavePlayer_69, properties {soundWaveAssetPtr=&ChainSaw_Scooter_Hit_Enemy_02}}
 {id=WavePlayer_7, properties {soundWaveAssetPtr=&ChainSaw_Impact_A_04}}
 {id=WavePlayer_70, properties {soundWaveAssetPtr=&ChainSaw_Scooter_Hit_Enemy_03}}
 {id=WavePlayer_71, properties {soundWaveAssetPtr=&ChainSaw_Scooter_Hit_Enemy_04}}
 {id=WavePlayer_8, properties {soundWaveAssetPtr=&ChainSaw_Impact_B_01}}
]
properties {
 attenuationSettings=/Game/Sounds/Attenuation/Fort_Weapon_Melee_Attack_Special_AOE_Slam_Att
 bOverrideConcurrency=true
 maxDistance=13031.375
 soundClassObject=&Fort_WeaponRanged_Fire_3P_PSM
 attenuationOverrides {bEnableReverbSend=false, distanceAlgorithm=Logarithmic, falloffDistance=1000, stereoSpread=100, attenuationShapeExtents {x=125}}
 busSends [{
   audioBus=&/AB_Chainsaw_Duck
   maxSendDistance=0
   minSendDistance=1000
   minSendLevel=0
   sendLevel=1
   soundSourceBus=null
  }]
 concurrencyOverrides {maxCount=3}
}
