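# Sound cue for the start of the chainsaw jump attack, layering five randomized sounds in one mixer.
# Blade swing and jump layers play through modulators without delay, while a modulated rev-up, sparks and spark bursts follow after short fixed delays.
# Each layer picks among equally weighted variations, with spark bursts mixed loudest.
# The cue uses close-range melee attenuation, overrides concurrency to limit simultaneous copies and sends to a chainsaw ducking bus.
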SoundCue /ChainsawGameplay/Assets/Sounds/ChainSaw_JumpAttackStart_Cue
root=Mixer_0
nodes [
 {id=Delay_0, children [Modulator_0], properties {delayMax=0.15, delayMin=0.15}}
 {id=Delay_1, children [Random_3], properties {delayMax=0.2, delayMin=0.2}}
 {id=Delay_3, children [Random_4], properties {delayMax=0.15, delayMin=0.15}}
 {id=Mixer_0, children [Delay_0,Delay_3,Modulator_4,Modulator_3,Delay_1], properties {inputVolume [0.7,1.5,0.4,1.1,0.4]}}
 {id=Modulator_0, children [Random_1]}
 {id=Modulator_3, children [Random_6]}
 {id=Modulator_4, children [Random_0]}
 {id=Random_0, children [WavePlayer_1, WavePlayer_0, WavePlayer_3, WavePlayer_4, WavePlayer_24], properties {weights [1,1,1,1,1]}}
 {id=Random_1, children [WavePlayer_7,WavePlayer_8], properties {weights [1,1]}}
 {id=Random_3, children [WavePlayer_9, WavePlayer_10, WavePlayer_13], properties {weights [1,1,1]}}
 {id=Random_4, children [WavePlayer_12, WavePlayer_16, WavePlayer_17], properties {weights [1,1,1]}}
 {id=Random_6, children [WavePlayer_18, WavePlayer_19, WavePlayer_20], properties {weights [1,1,1]}}
 {id=WavePlayer_0, properties {soundWaveAssetPtr=&/ChainSaw_Swing_Blade_02}}
 {id=WavePlayer_1, properties {soundWaveAssetPtr=&/ChainSaw_Swing_Blade_01}}
 {id=WavePlayer_10, properties {soundWaveAssetPtr=&/ChainSaw_Sparks_02}}
 {id=WavePlayer_12, properties {soundWaveAssetPtr=&/ChainSaw_Spark_Burst_01}}
 {id=WavePlayer_13, properties {soundWaveAssetPtr=&/ChainSaw_Sparks_04}}
 {id=WavePlayer_16, properties {soundWaveAssetPtr=&/ChainSaw_Spark_Burst_02}}
 {id=WavePlayer_17, properties {soundWaveAssetPtr=&/ChainSaw_Spark_Burst_03}}
 {id=WavePlayer_18, properties {soundWaveAssetPtr=&/ChainSaw_Jump_B_01}}
 {id=WavePlayer_19, properties {soundWaveAssetPtr=&/ChainSaw_Jump_B_02}}
 {id=WavePlayer_20, properties {soundWaveAssetPtr=&/ChainSaw_Jump_B_03}}
 {id=WavePlayer_24, properties {soundWaveAssetPtr=&/ChainSaw_Swing_Blade_05}}
 {id=WavePlayer_3, properties {soundWaveAssetPtr=&/ChainSaw_Swing_Blade_03}}
 {id=WavePlayer_4, properties {soundWaveAssetPtr=&/ChainSaw_Swing_Blade_04}}
 {id=WavePlayer_7, properties {soundWaveAssetPtr=&/ChainSaw_Attack_RevUps_01}}
 {id=WavePlayer_8, properties {soundWaveAssetPtr=&/ChainSaw_Attack_RevUps_02}}
 {id=WavePlayer_9, properties {soundWaveAssetPtr=&/ChainSaw_Sparks_01}}
]
properties {
 attenuationSettings=/Game/Sounds/Attenuation/Fort_Weapon_Melee_Attack_Close_Range_Stereo_Att
 bOverrideConcurrency=true
 maxDistance=10058.75
 soundClassObject=&Fort_WeaponRanged_Fire_3P_PSM
 volumeMultiplier=0.7
 attenuationOverrides {bEnableReverbSend=false, distanceAlgorithm=Logarithmic, falloffDistance=1000, stereoSpread=100, attenuationShapeExtents {x=125}}
 busSends [{
   audioBus=&/AB_Chainsaw_Duck
   maxSendDistance=0
   minSendDistance=1000
   minSendLevel=0
   sendLevel=1
   soundSourceBus=null
  }]
 concurrencyOverrides {maxCount=2}
}
