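# Sound cue for the chainsaw pickup sound in the ChainsawGameplay content.
# A single wave player feeds ChainSaw_Pickup_01 through a Modulator left at its default settings.
# The cue uses weapon-pickup attenuation, the loot-collect sound class and a reduced volume.
# It references the shared PickupMaximum_Oldest concurrency set; single-instance, prevent-new override values are also stored, with no override flag shown.
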SoundCue /ChainsawGameplay/Assets/Sounds/ChainSaw_Pickup_Cue
root=Modulator_0
nodes [
 {id=Modulator_0, children [WavePlayer_1]}
 {id=WavePlayer_1, properties {soundWaveAssetPtr=&/ChainSaw_Pickup_01}}
]
properties {
 attenuationSettings=&Fort_Weapon_Pickup_Att
 maxDistance=3500
 soundClassObject=&Fort_Loot_Collect
 volumeMultiplier=0.325
 concurrencyOverrides {maxCount=1, resolutionRule=PreventNew}
 concurrencySet [&PickupMaximum_Oldest]
}
