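# Sound cue for the chainsaw scooter's looping hit sound, built from idle-loop waves with close-range melee attenuation and capped concurrency.
# IsLocal and IsEnemy branches select one of three copies of the same chain: local or unset plays unscaled, enemy louder, non-enemy quieter.
# Each chain mixes two randomly picked looping layers through an enveloper that fades volume in and revs pitch up before settling, then adds slight random pitch variation.
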SoundCue /ChainsawGameplay/Assets/Sounds/ChainSaw_Scooter_LoopHit_Cue
root=Branch_0
nodes [
 {id=Branch_0, children [Modulator_0,Branch_1,Modulator_0], properties {boolParameterName=IsLocal}}
 {id=Branch_1, children [Mixer_8,Mixer_6,Modulator_0], properties {boolParameterName=IsEnemy}}
 {
  id=Enveloper_0
  children [Mixer_0]
  properties {
   pitchCurve {editorCurveData {keys [
      {interpMode=RCIM_Cubic, time=0}
      {arriveTangent=0.0033, interpMode=RCIM_Cubic, leaveTangent=0.0033, time=0.1, value=1.4165}
      <at 1.078 1.0036>
     ]}}
   volumeCurve {editorCurveData {keys [<at 0 0>,<at 0.4 1>,<at 3 1>,<at 8 1>]}}
  }
 }
 {
  id=Enveloper_3
  children [Mixer_1]
  properties {
   pitchCurve {editorCurveData {keys [
      {interpMode=RCIM_Cubic, time=0}
      {arriveTangent=0.0033, interpMode=RCIM_Cubic, leaveTangent=0.0033, time=0.1, value=1.4165}
      <at 1.078 1.0036>
     ]}}
   volumeCurve {editorCurveData {keys [<at 0 0>,<at 0.4 1>,<at 3 1>,<at 8 1>]}}
  }
 }
 {
  id=Enveloper_5
  children [Mixer_2]
  properties {
   pitchCurve {editorCurveData {keys [
      {interpMode=RCIM_Cubic, time=0}
      {arriveTangent=0.0033, interpMode=RCIM_Cubic, leaveTangent=0.0033, time=0.1, value=1.4165}
      <at 1.078 1.0036>
     ]}}
   volumeCurve {editorCurveData {keys [<at 0 0>,<at 0.4 1>,<at 3 1>,<at 8 1>]}}
  }
 }
 {id=Mixer_0, children [Random_1,Random_0], properties {inputVolume [0.3,1.25]}}
 {id=Mixer_1, children [Random_3,Random_2], properties {inputVolume [0.3,1.25]}}
 {id=Mixer_2, children [Random_5,Random_4], properties {inputVolume [0.3,1.25]}}
 {id=Mixer_6, children [Modulator_2], properties {inputVolume [0.7]}}
 {id=Mixer_8, children [Modulator_1], properties {inputVolume [1.5]}}
 {id=Modulator_0, children [Enveloper_0], properties {pitchMax=1.1, pitchMin=1}}
 {id=Modulator_1, children [Enveloper_3], properties {pitchMax=1.1, pitchMin=1}}
 {id=Modulator_2, children [Enveloper_5], properties {pitchMax=1.1, pitchMin=1}}
 {id=Random_0, children [WavePlayer_2,WavePlayer_3], properties {weights [1,1]}}
 {id=Random_1, children [WavePlayer_1,WavePlayer_4,WavePlayer_5,WavePlayer_0], properties {weights [1,1,1,1]}}
 {id=Random_2, children [WavePlayer_7,WavePlayer_8], properties {weights [1,1]}}
 {id=Random_3, children [WavePlayer_9, WavePlayer_10, WavePlayer_11, WavePlayer_6], properties {weights [1,1,1,1]}}
 {id=Random_4, children [WavePlayer_13, WavePlayer_14], properties {weights [1,1]}}
 {id=Random_5, children [WavePlayer_15, WavePlayer_16, WavePlayer_17, WavePlayer_12], properties {weights [1,1,1,1]}}
 {id=WavePlayer_0, properties {bLooping=true, soundWaveAssetPtr=&/ChainSaw_Scooter_Idle_Loop_04}}
 {id=WavePlayer_1, properties {bLooping=true, soundWaveAssetPtr=&/ChainSaw_Scooter_Idle_Loop_C_01}}
 {id=WavePlayer_10, properties {bLooping=true, soundWaveAssetPtr=&/ChainSaw_Scooter_Idle_Loop_C_02}}
 {id=WavePlayer_11, properties {bLooping=true, soundWaveAssetPtr=&/ChainSaw_Scooter_Idle_Loop_C_03}}
 {id=WavePlayer_12, properties {bLooping=true, soundWaveAssetPtr=&ChainSaw_Scooter_Idle_Loop_04}}
 {id=WavePlayer_13, properties {bLooping=true, soundWaveAssetPtr=&/ChainSaw_Scooter_Idle_Loop_05}}
 {id=WavePlayer_14, properties {bLooping=true, soundWaveAssetPtr=&/ChainSaw_Scooter_Idle_Loop_06}}
 {id=WavePlayer_15, properties {bLooping=true, soundWaveAssetPtr=&ChainSaw_Scooter_Idle_Loop_C_01}}
 {id=WavePlayer_16, properties {bLooping=true, soundWaveAssetPtr=&ChainSaw_Scooter_Idle_Loop_C_02}}
 {id=WavePlayer_17, properties {bLooping=true, soundWaveAssetPtr=&ChainSaw_Scooter_Idle_Loop_C_03}}
 {id=WavePlayer_2, properties {bLooping=true, soundWaveAssetPtr=&ChainSaw_Scooter_Idle_Loop_05}}
 {id=WavePlayer_3, properties {bLooping=true, soundWaveAssetPtr=&ChainSaw_Scooter_Idle_Loop_06}}
 {id=WavePlayer_4, properties {bLooping=true, soundWaveAssetPtr=&ChainSaw_Scooter_Idle_Loop_C_02}}
 {id=WavePlayer_5, properties {bLooping=true, soundWaveAssetPtr=&ChainSaw_Scooter_Idle_Loop_C_03}}
 {id=WavePlayer_6, properties {bLooping=true, soundWaveAssetPtr=&ChainSaw_Scooter_Idle_Loop_04}}
 {id=WavePlayer_7, properties {bLooping=true, soundWaveAssetPtr=&ChainSaw_Scooter_Idle_Loop_05}}
 {id=WavePlayer_8, properties {bLooping=true, soundWaveAssetPtr=&ChainSaw_Scooter_Idle_Loop_06}}
 {id=WavePlayer_9, properties {bLooping=true, soundWaveAssetPtr=&ChainSaw_Scooter_Idle_Loop_C_01}}
]
properties {
 attenuationSettings=/Game/Sounds/Attenuation/Fort_Weapon_Melee_Attack_Close_Range_Stereo_Att
 bOverrideConcurrency=true
 maxDistance=10058.75
 soundClassObject=&Fort_WeaponRanged_Fire_3P_PSM
 volumeMultiplier=0.95
 concurrencyOverrides {maxCount=3}
}
