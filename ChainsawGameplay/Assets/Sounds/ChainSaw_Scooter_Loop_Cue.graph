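# Sound cue for the chainsaw scooter's looping idle engine sound, with close-range melee attenuation and a small concurrency cap.
# IsLocal and IsEnemy branches select local, enemy or other-player mixes, with enemy boosted and local as the unset fallback.
# Each mix layers two staggered delays into envelopers that fade in volume and shape pitch for a spin-up.
# Each enveloper plays one of five randomly chosen looping waves, varying the engine tone.
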SoundCue /ChainsawGameplay/Assets/Sounds/ChainSaw_Scooter_Loop_Cue
root=Branch_0
nodes [
 {id=Branch_0, children [Mixer_0,Branch_1,Mixer_0], properties {boolParameterName=IsLocal}}
 {id=Branch_1, children [Mixer_8,Mixer_6,Mixer_0], properties {boolParameterName=IsEnemy}}
 {id=Delay_0, children [Enveloper_6], properties {delayMax=0.2, delayMin=0.2}}
 {id=Delay_1, children [Enveloper_7], properties {delayMax=0.5, delayMin=0.5}}
 {id=Delay_2, children [Enveloper_0], properties {delayMax=0.2, delayMin=0.2}}
 {id=Delay_3, children [Enveloper_1], properties {delayMax=0.5, delayMin=0.5}}
 {id=Delay_4, children [Enveloper_8], properties {delayMax=0.2, delayMin=0.2}}
 {id=Delay_5, children [Enveloper_9], properties {delayMax=0.5, delayMin=0.5}}
 {
  id=Enveloper_0
  children [Random_2]
  properties {
   pitchCurve {editorCurveData {keys [{interpMode=RCIM_Cubic, time=0}, {interpMode=RCIM_Cubic, leaveTangent=0.03, time=0.15, value=1.5}, <at 1 1.15>]}}
   volumeCurve {editorCurveData {keys [<at 0 0>,<at 0.3044 0.2419>,<at 0.6 1>,<at 3 1>,<at 8 1>]}}
  }
 }
 {
  id=Enveloper_1
  children [Random_1]
  properties {
   pitchCurve {editorCurveData {keys [{time=0}, <at 0.25 1.1 RCIM_Cubic>, {time=3}]}}
   volumeCurve {editorCurveData {keys [
      <at 0 0>
      <at 0.315 0.2414>
      {arriveTangent=0.2825, interpMode=RCIM_Cubic, leaveTangent=0.2825, time=0.568, value=0.7871}
      <at 3 1>
      <at 8 0.8>
     ]}}
  }
 }
 {
  id=Enveloper_6
  children [Random_4]
  properties {
   pitchCurve {editorCurveData {keys [{interpMode=RCIM_Cubic, time=0}, {interpMode=RCIM_Cubic, leaveTangent=0.03, time=0.15, value=1.5}, <at 1 1.15>]}}
   volumeCurve {editorCurveData {keys [<at 0 0>,<at 0.3044 0.2419>,<at 0.6 1>,<at 3 1>,<at 8 1>]}}
  }
 }
 {
  id=Enveloper_7
  children [Random_3]
  properties {
   pitchCurve {editorCurveData {keys [{time=0}, <at 0.25 1.1 RCIM_Cubic>, {time=3}]}}
   volumeCurve {editorCurveData {keys [
      <at 0 0>
      <at 0.315 0.2414>
      {arriveTangent=0.2825, interpMode=RCIM_Cubic, leaveTangent=0.2825, time=0.568, value=0.7871}
      <at 3 1>
      <at 8 0.8>
     ]}}
  }
 }
 {
  id=Enveloper_8
  children [Random_6]
  properties {
   pitchCurve {editorCurveData {keys [{interpMode=RCIM_Cubic, time=0}, {interpMode=RCIM_Cubic, leaveTangent=0.03, time=0.15, value=1.5}, <at 1 1.15>]}}
   volumeCurve {editorCurveData {keys [<at 0 0>,<at 0.3044 0.2419>,<at 0.6 1>,<at 3 1>,<at 8 1>]}}
  }
 }
 {
  id=Enveloper_9
  children [Random_5]
  properties {
   pitchCurve {editorCurveData {keys [{time=0}, <at 0.25 1.1 RCIM_Cubic>, {time=3}]}}
   volumeCurve {editorCurveData {keys [
      <at 0 0>
      <at 0.315 0.2414>
      {arriveTangent=0.2825, interpMode=RCIM_Cubic, leaveTangent=0.2825, time=0.568, value=0.7871}
      <at 3 1>
      <at 8 0.8>
     ]}}
  }
 }
 {id=Mixer_0, children [Delay_2,Delay_3], properties {inputVolume [0.7,0.85]}}
 {id=Mixer_1, children [Delay_0,Delay_1], properties {inputVolume [0.7,0.85]}}
 {id=Mixer_6, children [Delay_4,Delay_5], properties {inputVolume [0.7,0.85]}}
 {id=Mixer_8, children [Mixer_1], properties {inputVolume [1.5]}}
 {id=Random_1, children [WavePlayer_0,WavePlayer_2,WavePlayer_5,WavePlayer_6,WavePlayer_1], properties {weights [1,1,1,1,1]}}
 {id=Random_2, children [WavePlayer_3,WavePlayer_4,WavePlayer_7,WavePlayer_8,WavePlayer_9], properties {weights [1,1,1,1,1]}}
 {id=Random_3, children [WavePlayer_15, WavePlayer_16, WavePlayer_17, WavePlayer_18, WavePlayer_10], properties {weights [1,1,1,1,1]}}
 {id=Random_4, children [WavePlayer_11, WavePlayer_14, WavePlayer_19, WavePlayer_20, WavePlayer_22], properties {weights [1,1,1,1,1]}}
 {id=Random_5, children [WavePlayer_26, WavePlayer_27, WavePlayer_28, WavePlayer_29, WavePlayer_23], properties {weights [1,1,1,1,1]}}
 {id=Random_6, children [WavePlayer_24, WavePlayer_25, WavePlayer_30, WavePlayer_31, WavePlayer_32], properties {weights [1,1,1,1,1]}}
 {id=WavePlayer_0, properties {bLooping=true, soundWaveAssetPtr=&/ChainSaw_Scooter_Idle_Loop_A_01}}
 {id=WavePlayer_1, properties {bLooping=true, soundWaveAssetPtr=&/ChainSaw_Scooter_Idle_Loop_01}}
 {id=WavePlayer_10, properties {bLooping=true, soundWaveAssetPtr=&ChainSaw_Scooter_Idle_Loop_01}}
 {id=WavePlayer_11, properties {bLooping=true, soundWaveAssetPtr=&/ChainSaw_Scooter_Idle_Loop_02}}
 {id=WavePlayer_14, properties {bLooping=true, soundWaveAssetPtr=&/ChainSaw_Scooter_Idle_Loop_03}}
 {id=WavePlayer_15, properties {bLooping=true, soundWaveAssetPtr=&ChainSaw_Scooter_Idle_Loop_A_01}}
 {id=WavePlayer_16, properties {bLooping=true, soundWaveAssetPtr=&/ChainSaw_Scooter_Idle_Loop_A_02}}
 {id=WavePlayer_17, properties {bLooping=true, soundWaveAssetPtr=&/ChainSaw_Scooter_Idle_Loop_A_03}}
 {id=WavePlayer_18, properties {bLooping=true, soundWaveAssetPtr=&/ChainSaw_Scooter_Idle_Loop_A_04}}
 {id=WavePlayer_19, properties {bLooping=true, soundWaveAssetPtr=&/ChainSaw_Scooter_Idle_Loop_B_01}}
 {id=WavePlayer_2, properties {bLooping=true, soundWaveAssetPtr=&ChainSaw_Scooter_Idle_Loop_A_02}}
 {id=WavePlayer_20, properties {bLooping=true, soundWaveAssetPtr=&/ChainSaw_Scooter_Idle_Loop_B_02}}
 {id=WavePlayer_22, properties {bLooping=true, soundWaveAssetPtr=&/ChainSaw_Scooter_Idle_Loop_B_03}}
 {id=WavePlayer_23, properties {bLooping=true, soundWaveAssetPtr=&ChainSaw_Scooter_Idle_Loop_01}}
 {id=WavePlayer_24, properties {bLooping=true, soundWaveAssetPtr=&ChainSaw_Scooter_Idle_Loop_02}}
 {id=WavePlayer_25, properties {bLooping=true, soundWaveAssetPtr=&ChainSaw_Scooter_Idle_Loop_03}}
 {id=WavePlayer_26, properties {bLooping=true, soundWaveAssetPtr=&ChainSaw_Scooter_Idle_Loop_A_01}}
 {id=WavePlayer_27, properties {bLooping=true, soundWaveAssetPtr=&ChainSaw_Scooter_Idle_Loop_A_02}}
 {id=WavePlayer_28, properties {bLooping=true, soundWaveAssetPtr=&ChainSaw_Scooter_Idle_Loop_A_03}}
 {id=WavePlayer_29, properties {bLooping=true, soundWaveAssetPtr=&ChainSaw_Scooter_Idle_Loop_A_04}}
 {id=WavePlayer_3, properties {bLooping=true, soundWaveAssetPtr=&ChainSaw_Scooter_Idle_Loop_02}}
 {id=WavePlayer_30, properties {bLooping=true, soundWaveAssetPtr=&ChainSaw_Scooter_Idle_Loop_B_01}}
 {id=WavePlayer_31, properties {bLooping=true, soundWaveAssetPtr=&ChainSaw_Scooter_Idle_Loop_B_02}}
 {id=WavePlayer_32, properties {bLooping=true, soundWaveAssetPtr=&ChainSaw_Scooter_Idle_Loop_B_03}}
 {id=WavePlayer_4, properties {bLooping=true, soundWaveAssetPtr=&ChainSaw_Scooter_Idle_Loop_03}}
 {id=WavePlayer_5, properties {bLooping=true, soundWaveAssetPtr=&ChainSaw_Scooter_Idle_Loop_A_03}}
 {id=WavePlayer_6, properties {bLooping=true, soundWaveAssetPtr=&ChainSaw_Scooter_Idle_Loop_A_04}}
 {id=WavePlayer_7, properties {bLooping=true, soundWaveAssetPtr=&ChainSaw_Scooter_Idle_Loop_B_01}}
 {id=WavePlayer_8, properties {bLooping=true, soundWaveAssetPtr=&ChainSaw_Scooter_Idle_Loop_B_02}}
 {id=WavePlayer_9, properties {bLooping=true, soundWaveAssetPtr=&ChainSaw_Scooter_Idle_Loop_B_03}}
]
properties {
 attenuationSettings=/Game/Sounds/Attenuation/Fort_Weapon_Melee_Attack_Close_Range_Stereo_Att
 bOverrideConcurrency=true
 maxDistance=10058.75
 soundClassObject=&Fort_WeaponRanged_Fire_3P_PSM
 volumeMultiplier=0.65
 concurrencyOverrides {maxCount=3}
}
