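# Sound cue for the chainsaw scooter's overheat, built as a one-shot stack of randomized layers.
# A mixer plays engine-off (boosted) and scooter-jump variants immediately, with pressure release, sparks and spark bursts following after a short fixed delay.
# Each layer picks among equally weighted waves, and pressure release also passes through a modulator.
# It uses melee close-range attenuation with overrides, limits concurrency, and sends to a chainsaw ducking bus.
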SoundCue /ChainsawGameplay/Assets/Sounds/ChainSaw_Scooter_Overheat_Cue
root=Mixer_0
nodes [
 {id=Delay_0, children [Modulator_0], properties {delayMax=0.3, delayMin=0.3}}
 {id=Delay_1, children [Random_3], properties {delayMax=0.3, delayMin=0.3}}
 {id=Delay_2, children [Random_4], properties {delayMax=0.3, delayMin=0.3}}
 {id=Mixer_0, children [Delay_0,Delay_2,Random_7,Delay_1,Random_5], properties {inputVolume [1,0.5,0.5,0.8,1.5]}}
 {id=Modulator_0, children [Random_2]}
 {id=Random_2, children [WavePlayer_14, WavePlayer_15, WavePlayer_21, WavePlayer_22], properties {weights [1,1,1,1]}}
 {id=Random_3, children [WavePlayer_9, WavePlayer_10, WavePlayer_13], properties {weights [1,1,1]}}
 {id=Random_4, children [WavePlayer_12, WavePlayer_16, WavePlayer_17], properties {weights [1,1,1]}}
 {id=Random_5, children [WavePlayer_2, WavePlayer_5, WavePlayer_6, WavePlayer_11], properties {weights [1,1,1,1]}}
 {id=Random_7, children [WavePlayer_23, WavePlayer_25], properties {weights [1,1]}}
 {id=WavePlayer_10, properties {soundWaveAssetPtr=&/ChainSaw_Sparks_02}}
 {id=WavePlayer_11, properties {soundWaveAssetPtr=&/ChainSaw_EngineOff_04}}
 {id=WavePlayer_12, properties {soundWaveAssetPtr=&/ChainSaw_Spark_Burst_01}}
 {id=WavePlayer_13, properties {soundWaveAssetPtr=&/ChainSaw_Sparks_04}}
 {id=WavePlayer_14, properties {soundWaveAssetPtr=&/ChainSaw_PressureRelease_01}}
 {id=WavePlayer_15, properties {soundWaveAssetPtr=&/ChainSaw_PressureRelease_02}}
 {id=WavePlayer_16, properties {soundWaveAssetPtr=&/ChainSaw_Spark_Burst_02}}
 {id=WavePlayer_17, properties {soundWaveAssetPtr=&/ChainSaw_Spark_Burst_03}}
 {id=WavePlayer_2, properties {soundWaveAssetPtr=&/ChainSaw_EngineOff_01}}
 {id=WavePlayer_21, properties {soundWaveAssetPtr=&/ChainSaw_PressureRelease_03}}
 {id=WavePlayer_22, properties {soundWaveAssetPtr=&/ChainSaw_PressureRelease_04}}
 {id=WavePlayer_23, properties {soundWaveAssetPtr=&/ChainSaw_Scooter_Jump_01}}
 {id=WavePlayer_25, properties {soundWaveAssetPtr=&/ChainSaw_Scooter_Jump_02}}
 {id=WavePlayer_5, properties {soundWaveAssetPtr=&/ChainSaw_EngineOff_02}}
 {id=WavePlayer_6, properties {soundWaveAssetPtr=&/ChainSaw_EngineOff_03}}
 {id=WavePlayer_9, properties {soundWaveAssetPtr=&/ChainSaw_Sparks_01}}
]
properties {
 attenuationSettings=/Game/Sounds/Attenuation/Fort_Weapon_Melee_Attack_Close_Range_Stereo_Att
 bOverrideConcurrency=true
 maxDistance=10058.75
 soundClassObject=&Fort_WeaponRanged_Fire_3P_PSM
 volumeMultiplier=0.5
 attenuationOverrides {bEnableReverbSend=false, distanceAlgorithm=Logarithmic, falloffDistance=1000, stereoSpread=100, attenuationShapeExtents {x=125}}
 busSends [{
   audioBus=&/AB_Chainsaw_Duck
   maxSendDistance=0
   minSendDistance=1000
   minSendLevel=0
   sendLevel=1
   soundSourceBus=null
  }]
 concurrencyOverrides {maxCount=3, resolutionRule=StopFarthestThenPreventNew}
}
