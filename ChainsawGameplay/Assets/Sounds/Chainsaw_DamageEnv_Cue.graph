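# Sound cue for chainsaw environment-damage impacts, routed to the Fort_Impacts sound class with close-range melee attenuation.
# A mixer layers four random picks at separate volumes: two impact sets and two melee enemy-hit sets reused here.
# Only the melee A layer passes through a modulator for variation.
# A concurrency override caps simultaneous instances, and a bus send feeds a chainsaw ducking audio bus.
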SoundCue /ChainsawGameplay/Assets/Sounds/Chainsaw_DamageEnv_Cue
root=Mixer_0
nodes [
 {id=Mixer_0, children [Random_0,Random_2,Modulator_0,Random_3], properties {inputVolume [1.6,1,0.7,1.3]}}
 {id=Modulator_0, children [Random_1], properties {pitchMin=1}}
 {id=Random_0, children [WavePlayer_0,WavePlayer_1,WavePlayer_3,WavePlayer_4], properties {weights [1,1,1,1]}}
 {id=Random_1, children [WavePlayer_2,WavePlayer_5,WavePlayer_6,WavePlayer_7], properties {weights [1,1,1,1]}}
 {id=Random_2, children [WavePlayer_9, WavePlayer_10, WavePlayer_11], properties {weights [1,1,1]}}
 {
  id=Random_3
  children [WavePlayer_8, WavePlayer_12, WavePlayer_13, WavePlayer_14, WavePlayer_15, WavePlayer_16]
  properties {weights [1,1,1,1,1,1]}
 }
 {id=WavePlayer_0, properties {soundWaveAssetPtr=&/ChainSaw_Impact_C_01}}
 {id=WavePlayer_1, properties {soundWaveAssetPtr=&/ChainSaw_Impact_C_02}}
 {id=WavePlayer_10, properties {soundWaveAssetPtr=&/ChainSaw_Impact_D_02}}
 {id=WavePlayer_11, properties {soundWaveAssetPtr=&/ChainSaw_Impact_D_03}}
 {id=WavePlayer_12, properties {soundWaveAssetPtr=&/ChainSaw_Melee_Hit_Enemy_B_02}}
 {id=WavePlayer_13, properties {soundWaveAssetPtr=&/ChainSaw_Melee_Hit_Enemy_B_03}}
 {id=WavePlayer_14, properties {soundWaveAssetPtr=&/ChainSaw_Melee_Hit_Enemy_B_04}}
 {id=WavePlayer_15, properties {soundWaveAssetPtr=&/ChainSaw_Melee_Hit_Enemy_B_05}}
 {id=WavePlayer_16, properties {soundWaveAssetPtr=&/ChainSaw_Melee_Hit_Enemy_B_06}}
 {id=WavePlayer_2, properties {soundWaveAssetPtr=&/ChainSaw_Melee_Hit_Enemy_A_01}}
 {id=WavePlayer_3, properties {soundWaveAssetPtr=&/ChainSaw_Impact_C_03}}
 {id=WavePlayer_4, properties {soundWaveAssetPtr=&/ChainSaw_Impact_C_04}}
 {id=WavePlayer_5, properties {soundWaveAssetPtr=&/ChainSaw_Melee_Hit_Enemy_A_02}}
 {id=WavePlayer_6, properties {soundWaveAssetPtr=&/ChainSaw_Melee_Hit_Enemy_A_03}}
 {id=WavePlayer_7, properties {soundWaveAssetPtr=&/ChainSaw_Melee_Hit_Enemy_A_04}}
 {id=WavePlayer_8, properties {soundWaveAssetPtr=&/ChainSaw_Melee_Hit_Enemy_B_01}}
 {id=WavePlayer_9, properties {soundWaveAssetPtr=&/ChainSaw_Impact_D_01}}
]
properties {
 attenuationSettings=/Game/Sounds/Attenuation/Fort_Weapon_Melee_Attack_Close_Range_Stereo_Att
 bOverrideConcurrency=true
 maxDistance=10058.75
 soundClassObject=&Fort_Impacts
 volumeMultiplier=1.3
 busSends [{
   audioBus=&/AB_Chainsaw_Duck
   maxSendDistance=0
   minSendDistance=1000
   minSendLevel=0
   sendLevel=1
   soundSourceBus=null
  }]
 concurrencyOverrides {maxCount=3}
}
