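# Sound cue for equipping the chainsaw weapon.
# A mixer layers one fixed equip wave, boosted, with a quieter charge layer picked at random from three equally weighted variants.
# The random node is excluded from branch culling.
# Playback uses the chainsaw charge attenuation and a third-person ranged-weapon fire sound class, with overall volume slightly reduced.
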SoundCue /ChainsawGameplay/Assets/Sounds/Chainsaw_Equip_Cue
root=Mixer_0
nodes [
 {id=Mixer_0, children [WavePlayer_11, Random_0], properties {inputVolume [1.25,0.7]}}
 {id=Random_0, children [WavePlayer_10, WavePlayer_9, WavePlayer_8], properties {bShouldExcludeFromBranchCulling=true, weights [1,1,1]}}
 {id=WavePlayer_10, properties {soundWaveAssetPtr=&/ChainSaw_Equip_Charge__01}}
 {id=WavePlayer_11, properties {soundWaveAssetPtr=&/ChainSaw_Equip_01}}
 {id=WavePlayer_8, properties {soundWaveAssetPtr=&/ChainSaw_Equip_Charge__03}}
 {id=WavePlayer_9, properties {soundWaveAssetPtr=&/ChainSaw_Equip_Charge__02}}
]
properties {
 attenuationSettings=&/ChainSaw_Charge_Att
 maxDistance=2100
 soundClassObject=&Fort_WeaponRanged_Fire_3P_PSM
 volumeMultiplier=0.8
}
