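# Sound cue for environment hits by the Chainsaw Scooter, judging by its name, though it plays chainsaw melee enemy-hit waves.
# A Random node picks one of four equally weighted hit variations and feeds a Modulator that adds slight pitch variation.
# It uses close-range stereo melee attenuation and the impacts sound class, with a concurrency override limiting simultaneous instances to three.
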SoundCue /ChainsawGameplay/Assets/Sounds/Chainsaw_Scooter_EnvHit_Cue
root=Modulator_0
nodes [
 {id=Modulator_0, children [Random_1], properties {pitchMax=1.03, pitchMin=0.98}}
 {id=Random_1, children [WavePlayer_2,WavePlayer_5,WavePlayer_6,WavePlayer_7], properties {weights [1,1,1,1]}}
 {id=WavePlayer_2, properties {soundWaveAssetPtr=&/ChainSaw_Melee_Hit_Enemy_A_01}}
 {id=WavePlayer_5, properties {soundWaveAssetPtr=&/ChainSaw_Melee_Hit_Enemy_A_02}}
 {id=WavePlayer_6, properties {soundWaveAssetPtr=&/ChainSaw_Melee_Hit_Enemy_A_03}}
 {id=WavePlayer_7, properties {soundWaveAssetPtr=&/ChainSaw_Melee_Hit_Enemy_A_04}}
]
properties {
 attenuationSettings=/Game/Sounds/Attenuation/Fort_Weapon_Melee_Attack_Close_Range_Stereo_Att
 bOverrideConcurrency=true
 maxDistance=10058.75
 soundClassObject=&Fort_Impacts
 concurrencyOverrides {maxCount=3}
}
