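# Mono MetaSound source for the chainsaw's idle motor loop, spatialized with an idle attenuation asset and limited in concurrency.
# A wave player starts on play, with pitch and volume fed by a random-variation patch.
# An affiliation mix patch scales gain, making enemy chainsaws louder than non-enemy ones.
# An envelope follower on the AB_Chainsaw_Duck audio bus drives sidechain ducking, multiplying the output by one minus the envelope.
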MetaSoundSource /ChainsawGameplay/Assets/Sounds/MSS_Chainsaw_Idle_Loop
inputs [
 Character.Affiliation {type=Enum:FortAudioAffiliation}
 Character.Health {type=Float}
 Character.UniqueID {type=Int32}
 UE.Source.OnPlay {type=Trigger}
]
interfaces [Fort.Character:1.0, UE.OutputFormat.Mono:1.0, UE.Source:1.1]
outputs ["UE.OutputFormat.Mono.Audio:0" {type=Audio}]
pages [Default {
  input_defaults {Character.Affiliation=0, Character.Health=0, Character.UniqueID=0, UE.Source.OnPlay=null}
  nodes [
   n1 = /FortMetasound/Patches/MSP_RandomPitchTimeVolume() {also {defaults {"Random Pitch Max (Semitones)"=0}}}
   n2 = &MSP_AffiliationMixFloat(Affiliation=^Character.Affiliation) {also {defaults {Enemy=1.5, Non-Enemy=0.65}}}
   WavePlayer_0 = UE.Wave Player:Mono(Play=^UE.Source.OnPlay, Pitch Shift=@n1.Random Pitch) {also {defaults {Loop="<MetaSoundBuilderResult.SUCCEEDED: 0>", Wave Asset=&/ChainSaw_Idle_Loop_01}}}
   AudioBusReader1_0 = "UE.Audio Bus Reader (1)"() {also {defaults {Audio Bus=&/AB_Chainsaw_Duck}}}
   EnvelopeFollower_0 = UE.Envelope Follower(In=@AudioBusReader1_0.Out 0)
   Multiply_1 = UE.Multiply:Audio by Float(
    PrimaryOperand=UE.Multiply:Audio by Float(
     AdditionalOperands=@n1.Random Volume
     PrimaryOperand=UE.Multiply:Audio by Float(AdditionalOperands=@n2.Out, PrimaryOperand=@WavePlayer_0.Out Mono).Out
    ).Out
    AdditionalOperands=UE.Subtract:Float(AdditionalOperands=@EnvelopeFollower_0.Envelope) {also {defaults {PrimaryOperand=1}}}.Out
   )
  ]
  links [{dst="@UE.OutputFormat.Mono.Audio:0", src=Multiply_1.Out}]
 }]
properties {
 assetImportData=&self:AssetImportData
 attenuationSettings=&/ChainSaw_Motor_Idle_Att
 bLooping=true
 bOverrideConcurrency=true
 duration=10000
 maxDistance=3900
 sampleRate=48000
 soundClassObject=&Fort_WeaponRanged_Fire_3P_PSM
 volume=0.3
 concurrencyOverrides {maxCount=3}
}
