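# Looping mono MetaSound source for the Chainsaw scooter's surface sound, such as a dirt loop.
# The surface loop swapper patch crossfades between supplied waves on surface-change triggers, with pitch shift and a fade-out trigger.
# Its output is scaled by surface volume, by movement speed remapped and smoothed with faster rise than fall, and by an affiliation mix that makes enemies louder than non-enemies.
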MetaSoundSource /ChainsawGameplay/Assets/Sounds/MSS_Chainsaw_Surface_ScooterLoop
inputs [
 Character.Affiliation {type=Enum:FortAudioAffiliation}
 Character.Health {type=Float}
 Character.UniqueID {type=Int32}
 Movement.Speed {type=Float}
 Movement.SpeedAzimuth {type=Float}
 Movement.SpeedElevation {type=Float}
 Surfaces.CrossfadeTime {type=Time}
 Surfaces.FadeOut {type=Trigger}
 Surfaces.OnSurfaceChanged {type=Trigger}
 Surfaces.Pitch {type=Float}
 Surfaces.Sound {type=WaveAsset}
 Surfaces.Volume {type=Float}
 UE.Source.OnPlay {type=Trigger}
]
interfaces [Fort.Character:1.0, Fort.Movement:1.0, UE.OutputFormat.Mono:1.0, UE.Source:1.1]
outputs ["UE.OutputFormat.Mono.Audio:0" {type=Audio}]
pages [Default {
  input_defaults {
   Character.Affiliation=0
   Character.Health=0
   Character.UniqueID=0
   Movement.Speed=3840.0007
   Movement.SpeedAzimuth=0
   Movement.SpeedElevation=0
   Surfaces.CrossfadeTime=0.15
   Surfaces.FadeOut=null
   Surfaces.OnSurfaceChanged=null
   Surfaces.Pitch=0
   Surfaces.Sound=/Game/Sounds/Fort_Player/Surfaces/Dirt/Surface_Dirt_Loop_01
   Surfaces.Volume=1
   UE.Source.OnPlay=null
  }
  nodes [
   n1 = MapRange.MapRange:Float(In=^Movement.Speed) {also {defaults {In Range A=50, In Range B=3000}}}
   n2 = /FNGameplayCues/Sounds/Fort_Audio_Blueprints/SurfaceTypeSound/MSP_SurfaceLoopSwapper(
    CrossfadeTime=^Surfaces.CrossfadeTime
    FadeOutAllSounds=^Surfaces.FadeOut
    PlayNextSound=^Surfaces.OnSurfaceChanged
    PitchShift=^Surfaces.Pitch
    NextSoundToPlay=^Surfaces.Sound
   )
   Multiply_1 = UE.Multiply:Audio by Float(
    AdditionalOperands=&MSP_AffiliationMixFloat(Affiliation=^Character.Affiliation) {also {defaults {Enemy=1.2, Non-Enemy=0.5}}}.Out
    PrimaryOperand=UE.Multiply:Audio by Float(
     AdditionalOperands=&InterpSpeed(FloatInput=@n1.Out Value) {also {defaults {FallingInterpSpeed=0.5, RisingInterpSpeed=2}}}.FloatOutput
     PrimaryOperand=UE.Multiply:Audio by Float(AdditionalOperands=^Surfaces.Volume, PrimaryOperand=@n2.Out).Out
    ).Out
   )
  ]
  links [{dst="@UE.OutputFormat.Mono.Audio:0", src=Multiply_1.Out}]
 }]
properties {
 assetImportData=&self:AssetImportData
 attenuationSettings=/Game/Sounds/Attenuation/Fort_Weapon_Melee_Attack_Close_Range_Mono_Att
 bLooping=true
 bOverrideConcurrency=true
 duration=10000
 maxDistance=10058.75
 sampleRate=48000
 soundClassObject=&Fort_Gadgets
 volume=2.5
 concurrencyOverrides {maxCount=3}
}
