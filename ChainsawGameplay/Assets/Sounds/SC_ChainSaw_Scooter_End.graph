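# Sound cue for the end of the chainsaw scooter action, built from chainsaw jump samples in the melee weapon sound class.
# An IsLocal branch plays the local mix unattenuated and sends other cases through attenuation to an IsEnemy branch, which boosts enemies, lowers non-enemies and reuses the local mix when unset.
# Each path modulates a mix of three layers (A, B, C), each randomly picking one of three variants.
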SoundCue /ChainsawGameplay/Assets/Sounds/SC_ChainSaw_Scooter_End
root=Branch_3
nodes [
 {id=Attenuation_0, children [Branch_1], properties {attenuationSettings=&/ChainSaw_Charge_Att}}
 {id=Branch_1, children [Mixer_8,Mixer_6,Modulator_2], properties {boolParameterName=IsEnemy}}
 {id=Branch_3, children [Modulator_2, Attenuation_0, Attenuation_0], properties {boolParameterName=IsLocal}}
 {id=Mixer_1, children [Random_1,Random_4,Random_5], properties {inputVolume [0.3,0.5,1]}}
 {id=Mixer_2, children [Random_6,Random_7,Random_8], properties {inputVolume [0.3,0.5,1]}}
 {id=Mixer_3, children [Random_9,Random_10,Random_11], properties {inputVolume [0.3,0.5,1]}}
 {id=Mixer_6, children [Modulator_5], properties {inputVolume [0.7]}}
 {id=Mixer_8, children [Modulator_3], properties {inputVolume [1.25]}}
 {id=Modulator_2, children [Mixer_1]}
 {id=Modulator_3, children [Mixer_2]}
 {id=Modulator_5, children [Mixer_3]}
 {id=Random_1, children [WavePlayer_0,WavePlayer_6,WavePlayer_7], properties {weights [1,1,1]}}
 {id=Random_10, children [WavePlayer_34, WavePlayer_35, WavePlayer_36], properties {weights [1,1,1]}}
 {id=Random_11, children [WavePlayer_37, WavePlayer_38, WavePlayer_39], properties {weights [1,1,1]}}
 {id=Random_4, children [WavePlayer_8, WavePlayer_9, WavePlayer_10], properties {weights [1,1,1]}}
 {id=Random_5, children [WavePlayer_11, WavePlayer_12, WavePlayer_13], properties {weights [1,1,1]}}
 {id=Random_6, children [WavePlayer_22, WavePlayer_23, WavePlayer_24], properties {weights [1,1,1]}}
 {id=Random_7, children [WavePlayer_25, WavePlayer_26, WavePlayer_27], properties {weights [1,1,1]}}
 {id=Random_8, children [WavePlayer_28, WavePlayer_29, WavePlayer_30], properties {weights [1,1,1]}}
 {id=Random_9, children [WavePlayer_31, WavePlayer_32, WavePlayer_33], properties {weights [1,1,1]}}
 {id=WavePlayer_0, properties {soundWaveAssetPtr=&/ChainSaw_Jump_A_01}}
 {id=WavePlayer_10, properties {soundWaveAssetPtr=&/ChainSaw_Jump_B_03}}
 {id=WavePlayer_11, properties {soundWaveAssetPtr=&/ChainSaw_Jump_C_01}}
 {id=WavePlayer_12, properties {soundWaveAssetPtr=&/ChainSaw_Jump_C_02}}
 {id=WavePlayer_13, properties {soundWaveAssetPtr=&/ChainSaw_Jump_C_03}}
 {id=WavePlayer_22, properties {soundWaveAssetPtr=&ChainSaw_Jump_A_01}}
 {id=WavePlayer_23, properties {soundWaveAssetPtr=&/ChainSaw_Jump_A_02}}
 {id=WavePlayer_24, properties {soundWaveAssetPtr=&/ChainSaw_Jump_A_03}}
 {id=WavePlayer_25, properties {soundWaveAssetPtr=&/ChainSaw_Jump_B_01}}
 {id=WavePlayer_26, properties {soundWaveAssetPtr=&/ChainSaw_Jump_B_02}}
 {id=WavePlayer_27, properties {soundWaveAssetPtr=&ChainSaw_Jump_B_03}}
 {id=WavePlayer_28, properties {soundWaveAssetPtr=&ChainSaw_Jump_C_01}}
 {id=WavePlayer_29, properties {soundWaveAssetPtr=&ChainSaw_Jump_C_02}}
 {id=WavePlayer_30, properties {soundWaveAssetPtr=&ChainSaw_Jump_C_03}}
 {id=WavePlayer_31, properties {soundWaveAssetPtr=&ChainSaw_Jump_A_01}}
 {id=WavePlayer_32, properties {soundWaveAssetPtr=&ChainSaw_Jump_A_02}}
 {id=WavePlayer_33, properties {soundWaveAssetPtr=&ChainSaw_Jump_A_03}}
 {id=WavePlayer_34, properties {soundWaveAssetPtr=&ChainSaw_Jump_B_01}}
 {id=WavePlayer_35, properties {soundWaveAssetPtr=&ChainSaw_Jump_B_02}}
 {id=WavePlayer_36, properties {soundWaveAssetPtr=&ChainSaw_Jump_B_03}}
 {id=WavePlayer_37, properties {soundWaveAssetPtr=&ChainSaw_Jump_C_01}}
 {id=WavePlayer_38, properties {soundWaveAssetPtr=&ChainSaw_Jump_C_02}}
 {id=WavePlayer_39, properties {soundWaveAssetPtr=&ChainSaw_Jump_C_03}}
 {id=WavePlayer_6, properties {soundWaveAssetPtr=&ChainSaw_Jump_A_02}}
 {id=WavePlayer_7, properties {soundWaveAssetPtr=&ChainSaw_Jump_A_03}}
 {id=WavePlayer_8, properties {soundWaveAssetPtr=&ChainSaw_Jump_B_01}}
 {id=WavePlayer_9, properties {soundWaveAssetPtr=&ChainSaw_Jump_B_02}}
]
properties {
 maxDistance=2100
 soundClassObject=&Fort_WeaponMelee
 concurrencyOverrides {maxCount=1, resolutionRule=PreventNew}
 concurrencySet [&PickupMaximum_Oldest]
}
