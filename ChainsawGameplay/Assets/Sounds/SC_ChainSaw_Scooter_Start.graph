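# Sound cue for the start of the chainsaw scooter, played with close-range melee attenuation.
# IsLocal and IsEnemy branches select local, enemy or non-enemy copies of the same layered mix, with the enemy copy boosted through an extra mixer.
# Each mix layers a briefly delayed, modulated random Revup_Down under a random Jump_B and a later, pitched-up random Jump_A.
# A concurrency override allows one instance and prevents new ones.
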SoundCue /ChainsawGameplay/Assets/Sounds/SC_ChainSaw_Scooter_Start
root=Branch_0
nodes [
 {id=Branch_0, children [Mixer_12,Branch_1,Mixer_12], properties {boolParameterName=IsLocal}}
 {id=Branch_1, children [Mixer_8,Mixer_6,Mixer_12], properties {boolParameterName=IsEnemy}}
 {id=Delay_0, children [Modulator_1], properties {delayMax=0.15, delayMin=0.15}}
 {id=Delay_1, children [Modulator_3], properties {delayMax=0.15, delayMin=0.15}}
 {id=Delay_13, children [Random_17], properties {delayMax=0.05, delayMin=0.05}}
 {id=Delay_18, children [Modulator_25], properties {delayMax=0.15, delayMin=0.15}}
 {id=Delay_22, children [Random_31], properties {delayMax=0.05, delayMin=0.05}}
 {id=Delay_9, children [Random_11], properties {delayMax=0.05, delayMin=0.05}}
 {id=Mixer_0, children [Modulator_0,Delay_0], properties {inputVolume [1.25,1.25]}}
 {id=Mixer_1, children [Modulator_2,Delay_1], properties {inputVolume [1.25,1.25]}}
 {id=Mixer_12, children [Delay_22,Mixer_13], properties {inputVolume [0.7,2]}}
 {id=Mixer_13, children [Modulator_24,Delay_18], properties {inputVolume [1.25,1.25]}}
 {id=Mixer_15, children [Delay_9,Mixer_0], properties {inputVolume [0.7,2]}}
 {id=Mixer_6, children [Delay_13,Mixer_1], properties {inputVolume [0.7,2]}}
 {id=Mixer_8, children [Mixer_15], properties {inputVolume [1.25]}}
 {id=Modulator_0, children [Random_1]}
 {id=Modulator_1, children [Random_0], properties {pitchMax=1.15, pitchMin=1.1}}
 {id=Modulator_10, children [WavePlayer_35], properties {pitchMin=0.9, volumeMax=1.1, volumeMin=1}}
 {id=Modulator_11, children [WavePlayer_33], properties {pitchMin=0.9, volumeMax=1.1, volumeMin=1}}
 {id=Modulator_14, children [WavePlayer_58], properties {pitchMax=1.1, pitchMin=1.05, volumeMax=1.1, volumeMin=1}}
 {id=Modulator_15, children [WavePlayer_59], properties {pitchMin=0.9, volumeMax=1.1, volumeMin=1}}
 {id=Modulator_16, children [WavePlayer_57], properties {pitchMin=0.9, volumeMax=1.1, volumeMin=1}}
 {id=Modulator_2, children [Random_3]}
 {id=Modulator_24, children [Random_30]}
 {id=Modulator_25, children [Random_29], properties {pitchMax=1.15, pitchMin=1.1}}
 {id=Modulator_26, children [WavePlayer_106], properties {pitchMax=1.1, pitchMin=1.05, volumeMax=1.1, volumeMin=1}}
 {id=Modulator_27, children [WavePlayer_107], properties {pitchMin=0.9, volumeMax=1.1, volumeMin=1}}
 {id=Modulator_28, children [WavePlayer_105], properties {pitchMin=0.9, volumeMax=1.1, volumeMin=1}}
 {id=Modulator_3, children [Random_2], properties {pitchMax=1.15, pitchMin=1.1}}
 {id=Modulator_9, children [WavePlayer_34], properties {pitchMax=1.1, pitchMin=1.05, volumeMax=1.1, volumeMin=1}}
 {id=Random_0, children [WavePlayer_0,WavePlayer_1,WavePlayer_2], properties {bShouldExcludeFromBranchCulling=true, weights [1,1,1]}}
 {id=Random_1, children [WavePlayer_3,WavePlayer_4,WavePlayer_5], properties {bShouldExcludeFromBranchCulling=true, weights [1,1,1]}}
 {id=Random_11, children [Modulator_11,Modulator_9,Modulator_10], properties {bShouldExcludeFromBranchCulling=true, weights [1,1,1]}}
 {id=Random_17, children [Modulator_16,Modulator_14,Modulator_15], properties {bShouldExcludeFromBranchCulling=true, weights [1,1,1]}}
 {id=Random_2, children [WavePlayer_6,WavePlayer_7,WavePlayer_8], properties {bShouldExcludeFromBranchCulling=true, weights [1,1,1]}}
 {id=Random_29, children [WavePlayer_99, WavePlayer_100, WavePlayer_101], properties {bShouldExcludeFromBranchCulling=true, weights [1,1,1]}}
 {id=Random_3, children [WavePlayer_9, WavePlayer_10, WavePlayer_11], properties {bShouldExcludeFromBranchCulling=true, weights [1,1,1]}}
 {id=Random_30, children [WavePlayer_102, WavePlayer_103, WavePlayer_104], properties {bShouldExcludeFromBranchCulling=true, weights [1,1,1]}}
 {id=Random_31, children [Modulator_28,Modulator_26,Modulator_27], properties {bShouldExcludeFromBranchCulling=true, weights [1,1,1]}}
 {id=WavePlayer_0, properties {soundWaveAssetPtr=&/ChainSaw_Jump_A_03}}
 {id=WavePlayer_1, properties {soundWaveAssetPtr=&/ChainSaw_Jump_A_02}}
 {id=WavePlayer_10, properties {soundWaveAssetPtr=&/ChainSaw_Jump_B_02}}
 {id=WavePlayer_100, properties {soundWaveAssetPtr=&ChainSaw_Jump_A_02}}
 {id=WavePlayer_101, properties {soundWaveAssetPtr=&/ChainSaw_Jump_A_01}}
 {id=WavePlayer_102, properties {soundWaveAssetPtr=&/ChainSaw_Jump_B_01}}
 {id=WavePlayer_103, properties {soundWaveAssetPtr=&ChainSaw_Jump_B_02}}
 {id=WavePlayer_104, properties {soundWaveAssetPtr=&/ChainSaw_Jump_B_03}}
 {id=WavePlayer_105, properties {soundWaveAssetPtr=&/ChainSaw_Revup_Down_03}}
 {id=WavePlayer_106, properties {soundWaveAssetPtr=&/ChainSaw_Revup_Down_02}}
 {id=WavePlayer_107, properties {soundWaveAssetPtr=&/ChainSaw_Revup_Down_01}}
 {id=WavePlayer_11, properties {soundWaveAssetPtr=&ChainSaw_Jump_B_03}}
 {id=WavePlayer_2, properties {soundWaveAssetPtr=&ChainSaw_Jump_A_01}}
 {id=WavePlayer_3, properties {soundWaveAssetPtr=&ChainSaw_Jump_B_01}}
 {id=WavePlayer_33, properties {soundWaveAssetPtr=&ChainSaw_Revup_Down_03}}
 {id=WavePlayer_34, properties {soundWaveAssetPtr=&ChainSaw_Revup_Down_02}}
 {id=WavePlayer_35, properties {soundWaveAssetPtr=&ChainSaw_Revup_Down_01}}
 {id=WavePlayer_4, properties {soundWaveAssetPtr=&ChainSaw_Jump_B_02}}
 {id=WavePlayer_5, properties {soundWaveAssetPtr=&ChainSaw_Jump_B_03}}
 {id=WavePlayer_57, properties {soundWaveAssetPtr=&ChainSaw_Revup_Down_03}}
 {id=WavePlayer_58, properties {soundWaveAssetPtr=&ChainSaw_Revup_Down_02}}
 {id=WavePlayer_59, properties {soundWaveAssetPtr=&ChainSaw_Revup_Down_01}}
 {id=WavePlayer_6, properties {soundWaveAssetPtr=&ChainSaw_Jump_A_03}}
 {id=WavePlayer_7, properties {soundWaveAssetPtr=&ChainSaw_Jump_A_02}}
 {id=WavePlayer_8, properties {soundWaveAssetPtr=&ChainSaw_Jump_A_01}}
 {id=WavePlayer_9, properties {soundWaveAssetPtr=&ChainSaw_Jump_B_01}}
 {id=WavePlayer_99, properties {soundWaveAssetPtr=&ChainSaw_Jump_A_03}}
]
properties {
 attenuationSettings=/Game/Sounds/Attenuation/Fort_Weapon_Melee_Attack_Close_Range_Stereo_Att
 maxDistance=10058.75
 soundClassObject=&Fort_WeaponRanged_Fire_3P_PSM
 concurrencyOverrides {maxCount=1, resolutionRule=PreventNew}
 concurrencySet [&PickupMaximum_Oldest]
}
