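# Sound cue for the chainsaw scooter start landing, layering randomized impact and debris samples with close-range melee attenuation.
# IsLocal and IsEnemy branches select separate local, enemy and non-enemy mixes, with unset parameters falling back to the local mix.
# Each mix combines random A, B and C impacts and impact debris, pitch-modulating B and C in different ranges.
# The local mix adds a hit-debris layer, while the enemy path gets an extra volume boost.
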SoundCue /ChainsawGameplay/Assets/Sounds/SC_ChainSaw_Scooter_Start_Land
root=Branch_0
nodes [
 {id=Branch_0, children [Mixer_12,Branch_1,Mixer_12], properties {boolParameterName=IsLocal}}
 {id=Branch_1, children [Mixer_8,Mixer_6,Mixer_12], properties {boolParameterName=IsEnemy}}
 {id=Mixer_12, children [Mixer_14], properties {inputVolume [1.2]}}
 {id=Mixer_14, children [Modulator_29,Random_33,Modulator_30,Random_35,Random_1], properties {inputVolume [0.7,1.5,0.6,0.6,0.3]}}
 {id=Mixer_15, children [Mixer_5], properties {inputVolume [1.5]}}
 {id=Mixer_5, children [Modulator_12,Modulator_13,Random_15,Random_0], properties {inputVolume [0.7,0.6,0.6,0.3]}}
 {id=Mixer_6, children [Mixer_7], properties {inputVolume [1.5]}}
 {id=Mixer_7, children [Modulator_17,Modulator_18,Random_20,Random_2], properties {inputVolume [0.7,0.6,0.6,0.3]}}
 {id=Mixer_8, children [Mixer_15], properties {inputVolume [1.1]}}
 {id=Modulator_12, children [Random_12], properties {pitchMax=0.95, pitchMin=0.9}}
 {id=Modulator_13, children [Random_14], properties {pitchMax=1.15, pitchMin=1}}
 {id=Modulator_17, children [Random_18], properties {pitchMax=0.95, pitchMin=0.9}}
 {id=Modulator_18, children [Random_19], properties {pitchMax=1.15, pitchMin=1}}
 {id=Modulator_29, children [Random_32], properties {pitchMax=0.95, pitchMin=0.9}}
 {id=Modulator_30, children [Random_34], properties {pitchMax=1.15, pitchMin=1}}
 {id=Random_0, children [WavePlayer_0,WavePlayer_1,WavePlayer_3,WavePlayer_4], properties {weights [1,1,1,1]}}
 {id=Random_1, children [WavePlayer_2,WavePlayer_5,WavePlayer_6,WavePlayer_7], properties {weights [1,1,1,1]}}
 {id=Random_12, children [WavePlayer_36, WavePlayer_37, WavePlayer_38], properties {bShouldExcludeFromBranchCulling=true, weights [1,1,1]}}
 {id=Random_14, children [WavePlayer_46, WavePlayer_47, WavePlayer_48], properties {bShouldExcludeFromBranchCulling=true, weights [1,1,1]}}
 {
  id=Random_15
  children [WavePlayer_49, WavePlayer_50, WavePlayer_51, WavePlayer_52, WavePlayer_53]
  properties {bShouldExcludeFromBranchCulling=true, weights [1,1,1,1,1]}
 }
 {id=Random_18, children [WavePlayer_60, WavePlayer_61, WavePlayer_62], properties {bShouldExcludeFromBranchCulling=true, weights [1,1,1]}}
 {id=Random_19, children [WavePlayer_63, WavePlayer_64, WavePlayer_65], properties {bShouldExcludeFromBranchCulling=true, weights [1,1,1]}}
 {id=Random_2, children [WavePlayer_8, WavePlayer_9, WavePlayer_10, WavePlayer_11], properties {weights [1,1,1,1]}}
 {
  id=Random_20
  children [WavePlayer_66, WavePlayer_67, WavePlayer_68, WavePlayer_69, WavePlayer_70]
  properties {bShouldExcludeFromBranchCulling=true, weights [1,1,1,1,1]}
 }
 {id=Random_32, children [WavePlayer_108, WavePlayer_109, WavePlayer_110], properties {bShouldExcludeFromBranchCulling=true, weights [1,1,1]}}
 {
  id=Random_33
  children [WavePlayer_111, WavePlayer_112, WavePlayer_113, WavePlayer_114]
  properties {bShouldExcludeFromBranchCulling=true, weights [1,1,1,1]}
 }
 {id=Random_34, children [WavePlayer_115, WavePlayer_116, WavePlayer_117], properties {bShouldExcludeFromBranchCulling=true, weights [1,1,1]}}
 {
  id=Random_35
  children [WavePlayer_118, WavePlayer_119, WavePlayer_120, WavePlayer_121, WavePlayer_122]
  properties {bShouldExcludeFromBranchCulling=true, weights [1,1,1,1,1]}
 }
 {id=WavePlayer_0, properties {soundWaveAssetPtr=&/ChainSaw_Impact_A_01}}
 {id=WavePlayer_1, properties {soundWaveAssetPtr=&/ChainSaw_Impact_A_02}}
 {id=WavePlayer_10, properties {soundWaveAssetPtr=&ChainSaw_Impact_A_01}}
 {id=WavePlayer_108, properties {soundWaveAssetPtr=&/ChainSaw_Impact_B_01}}
 {id=WavePlayer_109, properties {soundWaveAssetPtr=&/ChainSaw_Impact_B_02}}
 {id=WavePlayer_11, properties {soundWaveAssetPtr=&/ChainSaw_Impact_A_04}}
 {id=WavePlayer_110, properties {soundWaveAssetPtr=&/ChainSaw_Impact_B_03}}
 {id=WavePlayer_111, properties {soundWaveAssetPtr=&/ChainSaw_Hit_Debris_01}}
 {id=WavePlayer_112, properties {soundWaveAssetPtr=&/ChainSaw_Hit_Debris_02}}
 {id=WavePlayer_113, properties {soundWaveAssetPtr=&/ChainSaw_Hit_Debris_03}}
 {id=WavePlayer_114, properties {soundWaveAssetPtr=&/ChainSaw_Hit_Debris_04}}
 {id=WavePlayer_115, properties {soundWaveAssetPtr=&/ChainSaw_Impact_C_02}}
 {id=WavePlayer_116, properties {soundWaveAssetPtr=&/ChainSaw_Impact_C_03}}
 {id=WavePlayer_117, properties {soundWaveAssetPtr=&/ChainSaw_Impact_C_04}}
 {id=WavePlayer_118, properties {soundWaveAssetPtr=&/ChainSaw_Impact_Debris_01}}
 {id=WavePlayer_119, properties {soundWaveAssetPtr=&/ChainSaw_Impact_Debris_02}}
 {id=WavePlayer_120, properties {soundWaveAssetPtr=&/ChainSaw_Impact_Debris_03}}
 {id=WavePlayer_121, properties {soundWaveAssetPtr=&/ChainSaw_Impact_Debris_06}}
 {id=WavePlayer_122, properties {soundWaveAssetPtr=&/ChainSaw_Impact_Debris_07}}
 {id=WavePlayer_2, properties {soundWaveAssetPtr=&ChainSaw_Impact_A_01}}
 {id=WavePlayer_3, properties {soundWaveAssetPtr=&ChainSaw_Impact_A_01}}
 {id=WavePlayer_36, properties {soundWaveAssetPtr=&ChainSaw_Impact_B_01}}
 {id=WavePlayer_37, properties {soundWaveAssetPtr=&ChainSaw_Impact_B_02}}
 {id=WavePlayer_38, properties {soundWaveAssetPtr=&ChainSaw_Impact_B_03}}
 {id=WavePlayer_4, properties {soundWaveAssetPtr=&ChainSaw_Impact_A_04}}
 {id=WavePlayer_46, properties {soundWaveAssetPtr=&ChainSaw_Impact_C_02}}
 {id=WavePlayer_47, properties {soundWaveAssetPtr=&ChainSaw_Impact_C_03}}
 {id=WavePlayer_48, properties {soundWaveAssetPtr=&ChainSaw_Impact_C_04}}
 {id=WavePlayer_49, properties {soundWaveAssetPtr=&ChainSaw_Impact_Debris_01}}
 {id=WavePlayer_5, properties {soundWaveAssetPtr=&ChainSaw_Impact_A_02}}
 {id=WavePlayer_50, properties {soundWaveAssetPtr=&ChainSaw_Impact_Debris_02}}
 {id=WavePlayer_51, properties {soundWaveAssetPtr=&ChainSaw_Impact_Debris_03}}
 {id=WavePlayer_52, properties {soundWaveAssetPtr=&ChainSaw_Impact_Debris_06}}
 {id=WavePlayer_53, properties {soundWaveAssetPtr=&ChainSaw_Impact_Debris_07}}
 {id=WavePlayer_6, properties {soundWaveAssetPtr=&ChainSaw_Impact_A_01}}
 {id=WavePlayer_60, properties {soundWaveAssetPtr=&ChainSaw_Impact_B_01}}
 {id=WavePlayer_61, properties {soundWaveAssetPtr=&ChainSaw_Impact_B_02}}
 {id=WavePlayer_62, properties {soundWaveAssetPtr=&ChainSaw_Impact_B_03}}
 {id=WavePlayer_63, properties {soundWaveAssetPtr=&ChainSaw_Impact_C_02}}
 {id=WavePlayer_64, properties {soundWaveAssetPtr=&ChainSaw_Impact_C_03}}
 {id=WavePlayer_65, properties {soundWaveAssetPtr=&ChainSaw_Impact_C_04}}
 {id=WavePlayer_66, properties {soundWaveAssetPtr=&ChainSaw_Impact_Debris_01}}
 {id=WavePlayer_67, properties {soundWaveAssetPtr=&ChainSaw_Impact_Debris_02}}
 {id=WavePlayer_68, properties {soundWaveAssetPtr=&ChainSaw_Impact_Debris_03}}
 {id=WavePlayer_69, properties {soundWaveAssetPtr=&ChainSaw_Impact_Debris_06}}
 {id=WavePlayer_7, properties {soundWaveAssetPtr=&ChainSaw_Impact_A_04}}
 {id=WavePlayer_70, properties {soundWaveAssetPtr=&ChainSaw_Impact_Debris_07}}
 {id=WavePlayer_8, properties {soundWaveAssetPtr=&ChainSaw_Impact_A_01}}
 {id=WavePlayer_9, properties {soundWaveAssetPtr=&ChainSaw_Impact_A_02}}
]
properties {
 attenuationSettings=/Game/Sounds/Attenuation/Fort_Weapon_Melee_Attack_Close_Range_Stereo_Att
 maxDistance=10058.75
 soundClassObject=&Fort_WeaponRanged_Fire_3P_PSM
 volumeMultiplier=0.9
 concurrencyOverrides {maxCount=1, resolutionRule=PreventNew}
 concurrencySet [&PickupMaximum_Oldest]
}
